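# Material function that adds an animated, glowing grid of random squares and lines to the BoldTouch character, with a separately rotated and offset variant for the shoes.
# Squares come from tiled UVs, hashed cell IDs and time-driven cosines, plus calls to MF_BoldTouch_RandomSquaresChar.
# Panning noise and tech-art masks shape the pattern.
# The result blends into BaseColor and adds to Emissive, boosted by a time-of-day leggings color in a masked region.
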MaterialFunction /BRCosmetics/Characters/Player/Female/Medium/Bodies/F_MED_BoldTouch/Materials/MF_BoldTouchCharacterGrid
graph {nodes [
  Add(A=NamedRerouteUsage(), B=Multiply(A=NamedRerouteUsage(), B=NamedRerouteUsage()))
  Add(A=NamedRerouteUsage(), B=Multiply(A=NamedRerouteUsage(), B=NamedRerouteUsage()))
  n1 = NamedRerouteUsage()
  n2 = NamedRerouteUsage()
  n3 = NamedRerouteUsage()
  n4 = ScalarParameter() {defaultValue=0.5, group=BoldTouchBackpackTechArt, parameterName=NoiseEdgeBreakupMult}
  n5 = Multiply(
   A=Add(A=Multiply(A=@n3, B=@n4), B=ScalarParameter() {defaultValue=0.25, group=BoldTouchBackpackTechArt, parameterName=NoiseEdgeBreakupBias})
   B=ScalarParameter() {defaultValue=2, group=BoldTouchBackpackTechArt, parameterName=MultiplyAfterPower}
  ) {constB=2}
  n6 = NamedRerouteUsage()
  n7 = Multiply(
   A=Multiply(A=Add(A=@n1, B=@n2), B=NamedRerouteDeclaration(None=@n5) {name=NoiseBreakup, nodeColor=<rgba 0 0.8359 1 1>})
   B=SmoothStep(
    Min=ScalarParameter() {group=BoldTouchBackpackTechArt, parameterName=MaskSmoothMin}
    Max=ScalarParameter() {defaultValue=1, group=BoldTouchBackpackTechArt, parameterName=MaskSmoothMax}
    Value=ComponentMask(None=@n6) {g=true}
   )
  )
  n8 = Add(A=NamedRerouteUsage(), B=NamedRerouteUsage())
  n9 = Multiply(A=NamedRerouteUsage(), B=ScalarParameter() {defaultValue=0.5, group=BoldTouchBackpackTechArt, parameterName=NoiseEdgeBreakupMult})
  n10 = ScalarParameter() {defaultValue=0.25, group=BoldTouchBackpackTechArt, parameterName=NoiseEdgeBreakupBias}
  n11 = NamedRerouteDeclaration(None=Multiply(A=Add(A=@n9, B=@n10), B=ScalarParameter() {defaultValue=2, group=BoldTouchBackpackTechArt, parameterName=MultiplyAfterPower}) {constB=2}) {name=NoiseBreakup_Shoes, nodeColor=<rgba 0 0.8359 1 1>}
  n12 = ScalarParameter() {group=BoldTouchBackpackTechArt, parameterName=MaskSmoothMin}
  n13 = ScalarParameter() {defaultValue=1, group=BoldTouchBackpackTechArt, parameterName=MaskSmoothMax}
  n14 = NamedRerouteUsage()
  Add(A=@n7, B=Multiply(A=Multiply(A=@n8, B=@n11), B=SmoothStep(Min=@n12, Max=@n13, Value=@n14)))
  n15 = GetMaterialAttributes(None=NamedRerouteUsage()) {attributeGetTypes [#1, #2]}
  n16 = NamedRerouteUsage()
  n17 = NamedRerouteUsage()
  n18 = NamedRerouteUsage()
  n19 = NamedRerouteUsage()
  n20 = NamedRerouteUsage()
  n21 = LinearInterpolate(
   A=@n7
   B=Multiply(
    A=@n7
    B=MaterialFunctionCall(Color_Morning=@n16, Color_Day=@n17, Color_Evening=@n18, Color_Night=@n19) {materialFunction=&Color_Time_of_Day}.Result
   ) {constB=12}
   Alpha=SmoothStep(Value=ComponentMask(None=@n20) {r=true}) {constMax=0.8, constMin=0.5}
  )
  n22 = NamedRerouteUsage()
  n23 = Multiply(A=@n7) {constB=4}
  FunctionOutput(None=SetMaterialAttributes!(
    MaterialAttributes=@n15.MaterialAttributes
    EmissiveColor=Add(A=@n15.EmissiveColor, B=@n21)
    BaseColor=LinearInterpolate(A=@n22, B=@n23, Alpha=@n7)
   ) {attributeSetTypes [#1, #2]})
  NamedRerouteDeclaration(None=FunctionInput() {bUsePreviewValueAsDefault=true, inputName=IN_TechArtMask, sortPriority=1}) {name=IN_TechArtMask, nodeColor=<rgba 1 0.375 0 1>}
  NamedRerouteDeclaration(None=FunctionInput() {bUsePreviewValueAsDefault=true, inputName=IN_MaterialAttributes, inputType=FunctionInput_MaterialAttributes}) {name=IN_MaterialAttributes, nodeColor=<rgba 0.0547 0 1 1>}
  NamedRerouteDeclaration(None=MaterialFunctionCall(Vector 3=Constant3Vector() {constant=<rgba 0 1 2 0>}) {materialFunction=/Engine/Functions/Engine_MaterialFunctions02/Utility/VectorLength}.V3 Length) {name=VectorLength, nodeColor=<rgba 0.1953 0 1 1>}
  n24 = NamedRerouteUsage()
  NamedRerouteDeclaration(None=Add(A=NamedRerouteUsage(), B=Multiply(B=@n24) {constA=2})) {name=TimePlusId, nodeColor=<rgba 0.125 1 0 1>}
  NamedRerouteDeclaration(None=Multiply(A=Time(), B=ScalarParameter() {defaultValue=0.2, group=BoldTouchBackpackTechArt, parameterName=TimeMultiplierShoes}) {constB=0.05}) {name=TimeMult_Shoes, nodeColor=<rgba 0.7578 0 1 1>}
  n25 = ScalarParameter() {defaultValue=0.75, group=BoldTouchBackpackTechArt, parameterName=RotateShoeSquares}
  n26 = Multiply(
   A=Rotator(
    Coordinate=TextureCoordinate()
    Time=MaterialFunctionCall(Multiplier=@n25) {materialFunction=&Pi}.Result
   )
   B=AppendVector(
    A=ScalarParameter() {defaultValue=24, group=BoldTouchBackpackTechArt, parameterName=UTilingGridLines_Shoes}
    B=ScalarParameter() {defaultValue=24, group=BoldTouchBackpackTechArt, parameterName=VTilingGridLines_Shoes, sortPriority=35}
   )
  )
  n27 = AppendVector(
   A=ScalarParameter() {group=BoldTouchBackpackTechArt, parameterName=OffsetShoeUV_X}
   B=ScalarParameter() {group=BoldTouchBackpackTechArt, parameterName=OffsetShoeUV_Y}
  )
  NamedRerouteDeclaration(None=Frac(None=Add(A=@n26, B=@n27))) {name=TiledUVsForGridLines_Shoes, nodeColor=<rgba 0.875 0 1 1>}
  n28 = Multiply(A=NamedRerouteUsage()) {constB=2}
  NamedRerouteDeclaration(None=Abs(None=Subtract(A=@n28))) {name=Square_Shoes, nodeColor=<rgba 0 1 0.5547 1>}
  NamedRerouteDeclaration(None=Step(Y=ScalarParameter() {defaultValue=0.95, group=BoldTouchBackpackTechArt, parameterName=GridLineSize_Shoes}, X=NamedRerouteUsage()) {constY=0.95}) {name=SharpSquare_Shoes, nodeColor=<rgba 1 0.2344 0 1>}
  n29 = TextureCoordinate()
  n30 = AppendVector(
   A=ScalarParameter() {defaultValue=24, group=BoldTouchBackpackTechArt, parameterName=UTilingGridLines}
   B=ScalarParameter() {defaultValue=24, group=BoldTouchBackpackTechArt, parameterName=VTilingGridLines, sortPriority=35}
  )
  NamedRerouteDeclaration(None=Frac(None=Multiply(A=@n29, B=@n30))) {name=TiledUVsForGridLines, nodeColor=<rgba 0.875 0 1 1>}
  n31 = Multiply(A=NamedRerouteUsage()) {constB=2}
  NamedRerouteDeclaration(None=Abs(None=Subtract(A=@n31))) {name=Square, nodeColor=<rgba 0 1 0.5547 1>}
  NamedRerouteDeclaration(None=Step(Y=ScalarParameter() {defaultValue=0.95, group=BoldTouchBackpackTechArt, parameterName=GridLineSize}, X=NamedRerouteUsage()) {constY=0.95}) {name=SharpSquare, nodeColor=<rgba 1 0.2344 0 1>}
  n32 = TextureCoordinate()
  n33 = AppendVector(
   A=ScalarParameter() {defaultValue=24, group=BoldTouchBackpackTechArt, parameterName=UTiling}
   B=ScalarParameter() {defaultValue=24, group=BoldTouchBackpackTechArt, parameterName=VTiling, sortPriority=35}
  )
  NamedRerouteDeclaration(None=Add(A=Multiply(A=@n32, B=@n33), B=Constant2Vector() {g=-0.5, r=-0.5})) {name=UVs, nodeColor=<rgba 1 0.8488 0 1>}
  n34 = MaterialFunctionCall(Float2=NamedRerouteUsage()) {materialFunction=&BreakOutFloat2Components}
  NamedRerouteDeclaration(None=Add(A=@n34.R, B=@n34.R)) {name=SquareResult, nodeColor=<rgba 0 0.0859 1 1>}
  n35 = MaterialFunctionCall(Float2=NamedRerouteUsage()) {materialFunction=&BreakOutFloat2Components}
  NamedRerouteDeclaration(None=Add(A=@n35.R, B=@n35.R)) {name=SquareResult_Shoes, nodeColor=<rgba 0 0.0859 1 1>}
  n36 = TextureCoordinate()
  n37 = Multiply(A=@n36)
  n38 = Time()
  n39 = AppendVector(
   A=ScalarParameter() {defaultValue=0.2, group=BoldTouchBackpackTechArt, parameterName=LowFreqNoiseUSpeed}
   B=ScalarParameter() {defaultValue=0.62, group=BoldTouchBackpackTechArt, parameterName=LowFreqNoiseVSpeed}
  )
  n40 = Multiply(A=@n36)
  n41 = Time()
  n42 = AppendVector(
   A=ScalarParameter() {defaultValue=0.2, group=BoldTouchBackpackTechArt, parameterName=HighFreqNoiseUSpeed}
   B=ScalarParameter() {defaultValue=0.62, group=BoldTouchBackpackTechArt, parameterName=HighFreqNoiseVSpeed}
  )
  n43 = Multiply(
   A=TextureSample(
    UVs=Panner(Coordinate=@n37, Time=@n38, Speed=@n39)
    Tex=TextureObjectParameter() {parameterName=NoiseTexture, texture=/BRCosmetics/Accessories/FORT_Backpacks/Backpack_F_MED_BoldTouch/Textures/T_PackedSquaresA}
   ).G
   B=TextureSample(
    UVs=Panner(Coordinate=@n40, Time=@n41, Speed=@n42)
    Tex=TextureObjectParameter() {parameterName=NoiseTextureB, texture=/BRCosmetics/Accessories/FORT_Backpacks/Backpack_F_MED_BoldTouch/Textures/T_PackedDotsA}
   ).B
  )
  NamedRerouteDeclaration(None=SmoothStep(
    Min=ScalarParameter() {group=BoldTouchBackpackTechArt, parameterName=MaskSStepMin}
    Max=ScalarParameter() {defaultValue=1, group=BoldTouchBackpackTechArt, parameterName=MaskSStepMax}
    Value=Saturate(None=@n43)
   )) {name=NoiseTotalForEdge, nodeColor=<rgba 0 1 0.6016 1>}
  n44 = Add(A=NamedRerouteUsage(), B=Constant2Vector() {g=0.5, r=0.5}) {constB=0.5}
  n45 = Sine(None=DotProduct(A=Floor(None=@n44), B=Constant2Vector() {g=17.81, r=113.1}))
  n46 = Cosine(None=Add(A=NamedRerouteUsage(), B=Multiply(A=@n45) {constB=43758.5469}))
  NamedRerouteDeclaration(None=Add(B=Multiply(B=@n46) {constA=0.5}) {constA=0.5}) {name=id_Shoes, nodeColor=<rgba 0 0.5078 1 1>}
  n47 = NamedRerouteUsage()
  n48 = Cosine(None=Divide(A=Add(A=@n47) {constB=0}, B=NamedRerouteUsage()))
  NamedRerouteDeclaration(None=Add(
    A=ScalarParameter() {defaultValue=0.25, parameterName=Amount of Lit Squares Shoes, sliderMax=1, sliderMin=-0.5}
    B=Multiply(B=@n48) {constA=0.5}
   ) {constA=-0.5}) {name=coRed_Shoes, nodeColor=<rgba 1 0 0.0044 1>}
  n49 = VectorParameter() {group=BoldTouchBackpackTechArt, parameterName=SquareColorA_Shoes, defaultValue=<rgba 0.7011 0.3419 0.8388 1>}
  n50 = VectorParameter() {group=BoldTouchBackpackTechArt, parameterName=SquareColorB_Shoes, defaultValue=<rgba 0.8963 0.2462 0.7991 1>}
  n51 = NamedRerouteUsage()
  NamedRerouteDeclaration(None=MaterialFunctionCall(
    A=Constant3Vector()
    B=@n49.RGB
    C=@n50.RGB
    Alpha=NamedRerouteDeclaration(None=@n51) {name=SquaresIntensity_Shoes, nodeColor=<rgba 1 0.7266 0 1>}
   ) {materialFunction=&Lerp_3Color}.Result) {name=Color_Shoes, nodeColor=<rgba 1 0.5156 0 1>}
  NamedRerouteDeclaration(None=MaterialFunctionCall(Vector 3=Constant3Vector() {constant=<rgba 0 1 2 0>}) {materialFunction=&VectorLength}.V3 Length) {name=VectorLength_Shoes, nodeColor=<rgba 0.1953 0 1 1>}
  n52 = NamedRerouteUsage()
  NamedRerouteDeclaration(None=Add(A=NamedRerouteUsage(), B=Multiply(B=@n52) {constA=2})) {name=TimePlusId_Shoes, nodeColor=<rgba 0.125 1 0 1>}
  n53 = Add(A=NamedRerouteUsage(), B=Constant2Vector() {g=0.5, r=0.5}) {constB=0.5}
  n54 = Sine(None=DotProduct(A=Floor(None=@n53), B=Constant2Vector() {g=17.81, r=113.1}))
  n55 = Cosine(None=Add(A=NamedRerouteUsage(), B=Multiply(A=@n54) {constB=43758.5469}))
  NamedRerouteDeclaration(None=Add(B=Multiply(B=@n55) {constA=0.5}) {constA=0.5}) {name=id, nodeColor=<rgba 0 0.5078 1 1>}
  NamedRerouteDeclaration(None=FunctionInput() {bUsePreviewValueAsDefault=true, inputName=IN_TechArtMask_Alpha, inputType=FunctionInput_Scalar, sortPriority=2}) {name=IN_TechArtMask_Alpha, nodeColor=<rgba 1 0.0028 0.612 1>}
  NamedRerouteDeclaration(None=@n15.BaseColor) {name=In_BaseColor, nodeColor=<rgba 1 0.7061 0 1>}
  NamedRerouteDeclaration(None=VectorParameter() {group=BoldTouchBackpackTechArt, parameterName=LeggingsSquaresDayColor, defaultValue=<rgba 12 12 12 0>}.RGB) {name=LeggingsSquaresColorDay, nodeColor=<rgba 1 0.0271 0 1>}
  NamedRerouteDeclaration(None=VectorParameter() {group=BoldTouchBackpackTechArt, parameterName=LeggingsSquaresEveningColor, defaultValue=<rgba 5 5 5 0>}.RGB) {name=LeggingsSquaresColorEvening, nodeColor=<rgba 0.556 0.0116 1 1>}
  NamedRerouteDeclaration(None=VectorParameter() {group=BoldTouchBackpackTechArt, parameterName=LeggingsSquaresNightColor, defaultValue=<rgba 4 4 4 0>}.RGB) {name=LeggingsSquaresColorNight, nodeColor=<rgba 0 1 0.9877 1>}
  n56 = NamedRerouteUsage()
  n57 = Multiply(B=SmoothStep(Value=ComponentMask(None=@n22) {r=true}) {constMax=0.9, constMin=0.5}) {constA=1}
  n58 = SmoothStep(
   Min=ScalarParameter() {group=BoldTouchBackpackTechArt, parameterName=MaskSmoothMin}
   Max=ScalarParameter() {defaultValue=1, group=BoldTouchBackpackTechArt, parameterName=MaskSmoothMax}
   Value=ComponentMask(None=NamedRerouteUsage()) {g=true}
  )
  NamedRerouteDeclaration(None=LinearInterpolate(A=NamedRerouteUsage(), B=Multiply(A=@n56) {constB=0.6}, Alpha=Multiply(A=@n57, B=@n58))) {name=BaseColorAdjustForLeggings, nodeColor=<rgba 0 0.25 1 1>}
  n59 = MaterialFunctionCall!(
   SquaresMax=ScalarParameter() {group=BoldTouchBackpackTechArt, parameterName=SquaresMinValue}
   SquaresMin=ScalarParameter() {defaultValue=1, group=BoldTouchBackpackTechArt, parameterName=SquaresMaxValue}
  ) {materialFunction=&/MF_BoldTouch_RandomSquaresChar}
  n60 = ScalarParameter() {parameterName="Min (S)"}
  n61 = ScalarParameter() {defaultValue=1, parameterName="Max (S)"}
  NamedRerouteDeclaration(None=MaterialFunctionCall(
    A=Constant3Vector()
    B=VectorParameter() {group=BoldTouchBackpackTechArt, parameterName=SquareColorA, defaultValue=<rgba 0.2423 0.0144 0.1946 1>}.RGB
    C=VectorParameter() {group=BoldTouchBackpackTechArt, parameterName=SquareColorB, defaultValue=<rgba 2 0.5494 1.7832 1>}.RGB
    Alpha=MaterialFunctionCall(Alpha=@n59.RandomSquares, Min=@n60, Max=@n61) {materialFunction=&SmoothStep}.Result
   ) {materialFunction=&Lerp_3Color}.Result) {name=RandomSquaresColor, nodeColor=<rgba 1 0.5156 0 1>}
  NamedRerouteDeclaration(None=Multiply(
    A=VectorParameter() {group=BoldTouchBackpackTechArt, parameterName=SquareLineColor, defaultValue=<rgba 1.5 0.412 1.3374 1>}.RGB
    B=MaterialFunctionCall() {materialFunction=&MF_BoldTouch_RandomSquaresChar}.Lines
   )) {name=LineColor, nodeColor=<rgba 1 0.5156 0 1>}
  n62 = NamedRerouteUsage()
  n63 = Cosine(None=Divide(A=Add(A=@n62) {constB=0}, B=NamedRerouteUsage()))
  NamedRerouteDeclaration(None=Add(A=ScalarParameter() {defaultValue=0.25, parameterName=Amount of Lit Squares, sliderMax=1, sliderMin=-0.5}, B=Multiply(B=@n63) {constA=0.5}) {constA=-0.5}) {name=coRed, nodeColor=<rgba 1 0 0.0044 1>}
  n64 = VectorParameter() {group=BoldTouchBackpackTechArt, parameterName=LeggingsSquaresColorMorning, defaultValue=<rgba 10 10 10 0>}
  NamedRerouteDeclaration(None=@n64.RGB) {name=LeggingsSquaresColorMorning, nodeColor=<rgba 0.5469 1 0 1>}
  NamedRerouteDeclaration(None=Multiply(A=Time(), B=ScalarParameter() {defaultValue=0.2, group=BoldTouchBackpackTechArt, parameterName=TimeMultiplier}) {constB=0.05}) {name=TimeMult, nodeColor=<rgba 0.7578 0 1 1>}
  n65 = ScalarParameter() {defaultValue=0.75, group=BoldTouchBackpackTechArt, parameterName=RotateShoeSquares}
  n66 = Multiply(
   A=Rotator(
    Coordinate=TextureCoordinate()
    Time=MaterialFunctionCall(Multiplier=@n65) {materialFunction=&Pi}.Result
   )
   B=AppendVector(
    A=ScalarParameter() {defaultValue=24, group=BoldTouchBackpackTechArt, parameterName=UTiling_Shoes}
    B=ScalarParameter() {defaultValue=24, group=BoldTouchBackpackTechArt, parameterName=VTiling_Shoes, sortPriority=35}
   )
  )
  n67 = AppendVector(
   A=ScalarParameter() {group=BoldTouchBackpackTechArt, parameterName=OffsetShoeUV_X}
   B=ScalarParameter() {group=BoldTouchBackpackTechArt, parameterName=OffsetShoeUV_Y}
  )
  NamedRerouteDeclaration(None=Add(A=Add(A=@n66, B=@n67), B=Constant2Vector() {g=-0.5, r=-0.5})) {name=UVs_shoes, nodeColor=<rgba 1 0.8488 0 1>}
  n68 = NamedRerouteUsage()
  NamedRerouteDeclaration(None=MaterialFunctionCall(
    A=Constant3Vector()
    B=VectorParameter() {group=BoldTouchBackpackTechArt, parameterName=SquareColorA, defaultValue=<rgba 0.7011 0.3419 0.8388 1>}.RGB
    C=VectorParameter() {group=BoldTouchBackpackTechArt, parameterName=SquareColorB, defaultValue=<rgba 0.8963 0.2462 0.7991 1>}.RGB
    Alpha=NamedRerouteDeclaration(None=@n68) {name=SquaresIntensity, nodeColor=<rgba 1 0.7266 0 1>}
   ) {materialFunction=&Lerp_3Color}.Result) {name=Color, nodeColor=<rgba 1 0.5156 0 1>}
  NamedRerouteUsage()
  Reroute(None=@n7)
 ]}
