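# Physics asset for the belt of the medium female BoldTouch player cosmetic, setting up a hanging strap on the left side.
# Front and back six-segment strap chains attach to the pelvis through base and offset bodies with free angular motion.
# Segments have tight swing limits and no twist, and an angularly locked constraint joins both tips into a loop.
# Pelvis, left thigh and calf capsules supply body shapes around the strap.
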PhysicsAsset /BRCosmetics/Characters/Player/Female/Medium/Bodies/F_MED_BoldTouch/Meshes/F_MED_BoldTouch_Belt_Physics
bodies {
 base_strap_back_lf [{shapeType=Capsule, length=0.505, radius=0.505, shapeName=base_strap_back_lf_capsule, rotation=<rot 0 -90 0>}]
 base_strap_front_lf [{shapeType=Capsule, length=0.505, radius=0.505, shapeName=base_strap_front_lf_capsule, rotation=<rot 0 -90 0>}]
 calf_l [{shapeType=Capsule, length=21.7748, radius=8.9476, center=<-21.0649 -2.8944 -0.9475>, rotation=<rot 0 87.5523 -88.6077>}]
 dyn_strap_back_01_lf [{shapeType=Capsule, length=3.164, radius=2, shapeName=dyn_strap_back_01_lf_capsule, center=<2 0 0>, rotation=<rot 90 0 0>}]
 dyn_strap_back_02_lf [{shapeType=Capsule, length=2.0659, radius=2, shapeName=dyn_strap_back_02_lf_capsule, center=<2 0 0>, rotation=<rot 90 0 0>}]
 dyn_strap_back_03_lf [{shapeType=Capsule, length=3.2295, radius=2, shapeName=dyn_strap_back_03_lf_capsule, center=<2 0 0>, rotation=<rot 90 0 0>}]
 dyn_strap_back_04_lf [{shapeType=Capsule, length=2.7542, radius=2, shapeName=dyn_strap_back_04_lf_capsule, center=<2 0 0>, rotation=<rot 90 0 0>}]
 dyn_strap_back_05_lf [{shapeType=Capsule, length=1.6382, radius=2.7904, shapeName=dyn_strap_back_05_lf_capsule, center=<2 0 0>, rotation=<rot 90 0 0>}]
 dyn_strap_back_06_lf [{shapeType=Capsule, length=1.2409, radius=3.0321, shapeName=dyn_strap_back_06_lf_capsule, center=<2 0 0>, rotation=<rot 90 0 0>}]
 dyn_strap_front_01_lf [{shapeType=Capsule, length=1.7803, radius=2, shapeName=dyn_strap_front_01_lf_capsule, center=<2 0 0>, rotation=<rot 90 0 0>}]
 dyn_strap_front_02_lf [{shapeType=Capsule, length=1.8251, radius=2, shapeName=dyn_strap_front_02_lf_capsule, center=<2 0 0>, rotation=<rot 90 0 0>}]
 dyn_strap_front_03_lf [{shapeType=Capsule, length=1.946, radius=2, shapeName=dyn_strap_front_03_lf_capsule, center=<2 0 0>, rotation=<rot 90 0 0>}]
 dyn_strap_front_04_lf [{shapeType=Capsule, length=2.9279, radius=2, shapeName=dyn_strap_front_04_lf_capsule, center=<2 0 0>, rotation=<rot 90 0 0>}]
 dyn_strap_front_05_lf [{shapeType=Capsule, length=1.7243, radius=3.0278, shapeName=dyn_strap_front_05_lf_capsule, center=<2 0 0>, rotation=<rot 90 0 0>}]
 dyn_strap_front_06_lf [{shapeType=Capsule, length=1.3904, radius=3.3078, shapeName=dyn_strap_front_06_lf_capsule, center=<2 0 0>, rotation=<rot 90 0 0>}]
 offset_strap_back_lf [{shapeType=Capsule, length=0.505, radius=0.505, shapeName=offset_strap_back_lf_capsule, rotation=<rot 0 -90 0>}]
 offset_strap_front_lf [{shapeType=Capsule, length=0.505, radius=0.505, shapeName=offset_strap_front_lf_capsule, rotation=<rot 0 -90 0>}]
 pelvis [{shapeType=Capsule, length=11, radius=9.1812, center=<-2.8163 -3.0233 -1.6413>}]
 thigh_l [{shapeType=Capsule, length=22.6361, radius=8.3726, center=<-22.0174 -1.7347 1.7645>, rotation=<rot 0 -90.6954 -82.5504>}]
}
constraints [
 {bone1Name=thigh_l, bone2Name=pelvis, swing1LimitDegrees=45, swing2LimitDegrees=45, twistLimitDegrees=45}
 {bone1Name=calf_l, bone2Name=thigh_l, swing1LimitDegrees=45, swing2LimitDegrees=45, twistLimitDegrees=45}
 {
  bone1Name=base_strap_back_lf
  bone2Name=pelvis
  swing1LimitDegrees=45
  swing1Motion=Free
  swing2LimitDegrees=45
  swing2Motion=Free
  twistLimitDegrees=45
  twistMotion=Free
 }
 {
  bone1Name=offset_strap_back_lf
  bone2Name=base_strap_back_lf
  swing1LimitDegrees=45
  swing1Motion=Free
  swing2LimitDegrees=45
  swing2Motion=Free
  twistLimitDegrees=45
  twistMotion=Free
 }
 {bone1Name=dyn_strap_back_01_lf, bone2Name=offset_strap_back_lf, swing1LimitDegrees=5, swing2LimitDegrees=10, twistLimitDegrees=0}
 {bone1Name=dyn_strap_back_02_lf, bone2Name=dyn_strap_back_01_lf, swing1LimitDegrees=5, swing2LimitDegrees=10, twistLimitDegrees=0}
 {bone1Name=dyn_strap_back_03_lf, bone2Name=dyn_strap_back_02_lf, swing1LimitDegrees=5, swing2LimitDegrees=10, twistLimitDegrees=0}
 {bone1Name=dyn_strap_back_04_lf, bone2Name=dyn_strap_back_03_lf, swing1LimitDegrees=5, swing2LimitDegrees=10, twistLimitDegrees=0}
 {bone1Name=dyn_strap_back_05_lf, bone2Name=dyn_strap_back_04_lf, swing1LimitDegrees=5, swing2LimitDegrees=10, twistLimitDegrees=0}
 {bone1Name=dyn_strap_back_06_lf, bone2Name=dyn_strap_back_05_lf, swing1LimitDegrees=5, swing2LimitDegrees=10, twistLimitDegrees=0}
 {
  bone1Name=base_strap_front_lf
  bone2Name=pelvis
  swing1LimitDegrees=45
  swing1Motion=Free
  swing2LimitDegrees=45
  swing2Motion=Free
  twistLimitDegrees=45
  twistMotion=Free
 }
 {
  bone1Name=offset_strap_front_lf
  bone2Name=base_strap_front_lf
  swing1LimitDegrees=45
  swing1Motion=Free
  swing2LimitDegrees=45
  swing2Motion=Free
  twistLimitDegrees=45
  twistMotion=Free
 }
 {bone1Name=dyn_strap_front_01_lf, bone2Name=offset_strap_front_lf, swing1LimitDegrees=5, swing2LimitDegrees=10, twistLimitDegrees=0}
 {bone1Name=dyn_strap_front_02_lf, bone2Name=dyn_strap_front_01_lf, swing1LimitDegrees=5, swing2LimitDegrees=10, twistLimitDegrees=0}
 {bone1Name=dyn_strap_front_03_lf, bone2Name=dyn_strap_front_02_lf, swing1LimitDegrees=5, swing2LimitDegrees=10, twistLimitDegrees=0}
 {bone1Name=dyn_strap_front_04_lf, bone2Name=dyn_strap_front_03_lf, swing1LimitDegrees=5, swing2LimitDegrees=10, twistLimitDegrees=0}
 {bone1Name=dyn_strap_front_05_lf, bone2Name=dyn_strap_front_04_lf, swing1LimitDegrees=5, swing2LimitDegrees=10, twistLimitDegrees=0}
 {bone1Name=dyn_strap_front_06_lf, bone2Name=dyn_strap_front_05_lf, swing1LimitDegrees=5, swing2LimitDegrees=10, twistLimitDegrees=0}
 {
  bone1Name=dyn_strap_front_06_lf
  bone2Name=dyn_strap_back_06_lf
  swing1LimitDegrees=45
  swing1Motion=Locked
  swing2LimitDegrees=45
  swing2Motion=Locked
  twistLimitDegrees=45
  twistMotion=Locked
 }
]
