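# Control Rig for the F_MED_BoldTouch female body that drives corrective and accessory bones for its jacket, belts, straps and shoes.
# It runs a shared DeformRig function, then uses weighted parent and aim constraints plus Jacket and shoe functions.
# Spherical pose readers on limbs and pelvis weight corrective transforms, some scaled by spine-to-calf distance.
# Left-thigh belt logic splits swing from twist and filters it by direction to drive straps.
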ControlRigBlueprint /BRCosmetics/Characters/Player/Female/Medium/Bodies/F_MED_BoldTouch/Meshes/F_MED_BoldTouch_CtrlRig
graphs [
 CollapseNode_ContainedGraph {nodes [
   n1 = RigVMFunctionEntryNode()
   &RigUnit_ParentConstraint(
    AdvancedSettings=(InterpolationType=Average,RotationOrderForFilter=XZY)
    Child=(Type=Bone,Name="thigh_l_get_rot")
    Parents=((Item=(Type=Bone,Name="thigh_l"),Weight=1))
    Weight=1
    ExecutePin=@n1.ExecutePin
   )
   > get_swing_axis_front = RigVMCollapseNode(
    Axis to Get=(X=0,Y=1,Z=1)
    Backward=false
    Forward=true
    Inward=false
    Outward=true
    Rotation Order=ZXY
    Twist Axis=(X=1,Y=0,Z=0)
    Twist_Inward=false
    Twist_Outward=false
    input joint=thigh_l_get_rot
   )
   > &RigUnit_ModifyTransforms(
    ItemToModify=((Item=(Type=Bone,Name="thigh_l_set_rot_front"),Transform=(Rotation=(X=0,Y=0,Z=0,W=1),Translation=(X=0,Y=0,Z=0),Scale3D=(X=1,Y=1,Z=1))))
    Weight=1
    WeightMinimum=0
    ItemToModify.0.Transform.Rotation=@get_swing_axis_front.Output Rotate.Value
   )
   > &RigUnit_RotationConstraintLocalSpaceOffset(
    AdvancedSettings=(InterpolationType=Average,RotationOrderForFilter=XYZ)
    Child=(Type=Bone,Name="base_strap_front_lf")
    Filter=(bX=True,bY=True,bZ=True)
    Parents=((Item=(Type=Bone,Name="thigh_l_set_rot_front"),Weight=1))
    Weight=1
   )
   > &RigUnit_RotationConstraintLocalSpaceOffset(
    AdvancedSettings=(InterpolationType=Average,RotationOrderForFilter=XYZ)
    Child=(Type=Bone,Name="offset_strap_front_second_lf")
    Filter=(bX=True,bY=True,bZ=True)
    Parents=((Item=(Type=Bone,Name="thigh_l_set_rot_front"),Weight=1))
    Weight=1
   )
   RigVMFunction_MathQuaternionToEuler = &RigVMFunction_MathQuaternionToEuler(RotationOrder=ZXY, Value=@get_swing_axis_front.Output Rotate.Value)
   n2 = &RigVMFunction_MathDoubleRemap(SourceMaximum=-110, SourceMinimum=0, TargetMaximum=1, TargetMinimum=0, Value=@RigVMFunction_MathQuaternionToEuler.Result.Z)
   > &RigUnit_ModifyTransforms(
    ItemToModify=((Item=(Type=Bone,Name="base_belt_lf"),Transform=(Rotation=(X=0,Y=0,Z=0,W=1),Translation=(X=0,Y=0,Z=1),Scale3D=(X=1,Y=1,Z=1))))
    WeightMinimum=0
    Weight=&RigVMFunction_MathDoubleRemap(SourceMaximum=0.2, SourceMinimum=0, TargetMaximum=1, TargetMinimum=0, Value=@n2.Result).Result
   )
   > &RigUnit_ModifyTransforms(
    ItemToModify="((Item=(Type=Bone,Name=\"belt_lf\"),Transform=(Rotation=(X=-0.0057,Y=-0.1294,Z=-0.1304,W=0.983),Translation=(X=0,Y=0,Z=0),Scale3D=(X=1,Y=1,Z=1))),(Item=(Type=Bone,Name=\"base_belt_lf\"),Transform=(Rotation=(X=0,Y=0,Z=0,W=1),Translation=(X=1.9111,Y=0,Z=0),Scale ...(truncated)"
    WeightMinimum=0
    Weight=&RigVMFunction_MathDoubleRemap(SourceMaximum=0.6, SourceMinimum=0.4, TargetMaximum=1, TargetMinimum=0, Value=@n2.Result).Result
   )
   > &RigUnit_ModifyTransforms(
    ItemToModify="((Item=(Type=Bone,Name=\"offset_strap_front_lf\"),Transform=(Rotation=(X=0.1206,Y=-0.3794,Z=-0.05,W=-0.916),Translation=(X=0,Y=0,Z=0),Scale3D=(X=1,Y=1,Z=1))),(Item=(Type=Bone,Name=\"belt_lf\"),Transform=(Rotation=(X=-0.1717,Y=0.8234,Z=0.1318,W=-0.5246),Transla ...(truncated)"
    WeightMinimum=0
    Weight=&RigVMFunction_MathDoubleRemap(SourceMaximum=1, SourceMinimum=0.4, TargetMaximum=1, TargetMinimum=0, Value=@n2.Result).Result
   )
   > &RigUnit_ModifyTransforms(
    ItemToModify="((Item=(Type=Bone,Name=\"belt_lf\"),Transform=(Rotation=(X=-0.2579,Y=-0.4728,Z=-0.2956,W=0.7891),Translation=(X=-0.8204,Y=-3.2752,Z=-0.665),Scale3D=(X=1,Y=1,Z=1))),(Item=(Type=Bone,Name=\"base_belt_lf\"),Transform=(Rotation=(X=0,Y=0,Z=0,W=1),Translation=(X=0,Y ...(truncated)"
    WeightMinimum=0
    Weight=@n1.Weight
   )
   > n3 = &RigUnit_SphericalPoseReader(
    ActiveRegionScaleFactors=(PositiveWidth=1,NegativeWidth=1,PositiveHeight=1,NegativeHeight=1)
    ActiveRegionSize=0.07
    Debug=(bDrawDebug=false,bDraw2D=False,bDrawLocalAxes=False,DebugScale=25,DebugSegments=20,DebugThickness=0.25)
    DriverAxis=(X=-1,Y=0,Z=0)
    DriverItem=(Type=Bone,Name="thigh_l")
    FalloffRegionScaleFactors=(PositiveWidth=1,NegativeWidth=1,PositiveHeight=1,NegativeHeight=1)
    FalloffSize=0.2
    FlipHeightScaling=False
    FlipWidthScaling=False
    OptionalParentItem=(Type=Bone,Name="None")
    RotationOffset=(X=130,Y=0,Z=-20)
   )
   > &RigUnit_ModifyTransforms(
    ItemToModify=((Item=(Type=Bone,Name="dyn_strap_front_01_lf"),Transform=(Rotation=(X=0,Y=0.0872,Z=0,W=-0.9962),Translation=(X=0,Y=0,Z=0),Scale3D=(X=1,Y=1,Z=1))))
    WeightMinimum=0
    Weight=@n3.OutputParam
   )
   > get_swing_axis_back = RigVMCollapseNode(
    Axis to Get=(X=0,Y=1,Z=1)
    Backward=true
    Forward=false
    Inward=false
    Outward=true
    Rotation Order=ZXY
    Twist Axis=(X=1,Y=0,Z=0)
    Twist_Inward=false
    Twist_Outward=false
    input joint=thigh_l_get_rot
   )
   > &RigUnit_ModifyTransforms(
    ItemToModify=((Item=(Type=Bone,Name="thigh_l_set_rot_back"),Transform=(Rotation=(X=0,Y=0,Z=0,W=1),Translation=(X=0,Y=0,Z=0),Scale3D=(X=1,Y=1,Z=1))))
    Weight=1
    WeightMinimum=0
    ItemToModify.0.Transform.Rotation=@get_swing_axis_back.Output Rotate.Value
   )
   > &RigUnit_RotationConstraintLocalSpaceOffset(
    AdvancedSettings=(InterpolationType=Average,RotationOrderForFilter=XYZ)
    Child=(Type=Bone,Name="offset_strap_back_lf")
    Filter=(bX=True,bY=True,bZ=True)
    Parents=((Item=(Type=Bone,Name="thigh_l_set_rot_back"),Weight=1))
    Weight=1
   )
   n4 = &RigVMFunction_MathQuaternionToEuler(RotationOrder=ZXY, Value=@get_swing_axis_back.Output Rotate.Value)
   > n5 = &RigUnit_ModifyTransforms(
    ItemToModify=((Item=(Type=Bone,Name="dyn_strap_back_01_lf"),Transform=(Rotation=(X=0,Y=0,Z=0,W=1),Translation=(X=2,Y=0,Z=0),Scale3D=(X=1,Y=1,Z=1))))
    WeightMinimum=0
    Weight=&RigVMFunction_MathDoubleRemap(SourceMaximum=70, SourceMinimum=0, TargetMaximum=1, TargetMinimum=0, Value=@n4.Result.Z).Result
   )
   RigVMFunctionReturnNode(ExecutePin=@n5.ExecutePin)
   &RigVMFunction_MathDoubleRemap(SourceMaximum=70, SourceMinimum=0, TargetMaximum=1, TargetMinimum=0, Value=@n4.Result.Z)
  ]}
 New Function_ContainedGraph {nodes [
   RigVMFunctionEntryNode()
   > RigVMFunction_Sequence_1 = RigVMAggregateNode()
   &RigUnit_AimConstraintLocalSpaceOffset(
    Child=(Type=Bone,Name="base_jacket_spine_cr")
    Filter=(bX=True,bY=True,bZ=True)
    Parents=((Item=(Type=Bone,Name="pelvis"),Weight=1))
    UpAxis=(X=0,Y=1,Z=0)
    Weight=1
    WorldUp=(Target=(X=0,Y=1,Z=0),Kind=Direction,Space=(Type=Bone,Name="base_jacket_spine_cr"))
    ExecutePin=@RigVMFunction_Sequence_1.A
   )
   > n1 = &RigUnit_SphericalPoseReader(
    ActiveRegionScaleFactors=(PositiveWidth=1,NegativeWidth=1,PositiveHeight=1,NegativeHeight=1)
    ActiveRegionSize=0.2
    Debug=(bDrawDebug=false,bDraw2D=False,bDrawLocalAxes=False,DebugScale=25,DebugSegments=20,DebugThickness=0.25)
    DriverAxis=(X=0,Y=1,Z=0)
    DriverItem=(Type=Bone,Name="pelvis")
    FalloffRegionScaleFactors=(PositiveWidth=1,NegativeWidth=1,PositiveHeight=1,NegativeHeight=1)
    FalloffSize=0.4
    FlipHeightScaling=False
    FlipWidthScaling=False
    OptionalParentItem=(Type=Bone,Name="spine_03")
    RotationOffset=(X=90,Y=0,Z=-90)
   )
   > &RigUnit_ModifyTransforms(
    ItemToModify="((Item=(Type=Bone,Name=\"base_jacket_front_rt\"),Transform=(Rotation=(X=0,Y=-0.0872,Z=0,W=0.9962),Translation=(X=0,Y=0,Z=0),Scale3D=(X=1,Y=1,Z=1))),(Item=(Type=Bone,Name=\"base_jacket_front_lf\"),Transform=(Rotation=(X=0,Y=-0.0872,Z=0,W=0.9962),Translation=(X= ...(truncated)"
    WeightMinimum=0
    Weight=@n1.OutputParam
   )
   > n2 = &RigUnit_SphericalPoseReader(
    ActiveRegionScaleFactors=(PositiveWidth=1,NegativeWidth=1,PositiveHeight=1,NegativeHeight=1)
    ActiveRegionSize=0.15
    Debug=(bDrawDebug=true,bDraw2D=False,bDrawLocalAxes=False,DebugScale=25,DebugSegments=20,DebugThickness=0.25)
    DriverAxis=(X=1,Y=0,Z=0)
    DriverItem=(Type=Bone,Name="upperarm_l")
    FalloffRegionScaleFactors=(PositiveWidth=1,NegativeWidth=1,PositiveHeight=1,NegativeHeight=1)
    FalloffSize=0.25
    FlipHeightScaling=False
    FlipWidthScaling=false
    OptionalParentItem=(Type=Bone,Name="None")
    RotationOffset=(X=0,Y=-45,Z=20)
   )
   > &RigUnit_ParentConstraint(
    AdvancedSettings=(InterpolationType=Average,RotationOrderForFilter=XZY)
    Child=(Type=Bone,Name="spine_03_for_cons")
    Parents=((Item=(Type=Bone,Name="spine_03"),Weight=1))
    Weight=1
   )
   n3 = &RigVMFunction_AlphaInterp(InterpSpeedDecreasing=10, InterpSpeedIncreasing=10, Scale=1, Value=1, bClampResult=False)
   n4 = &RigVMFunction_AnimEvalRichCurve(
    Curve="(EditorCurveData=(Keys=((InterpMode=RCIM_Cubic,TangentMode=RCTM_Auto,TangentWeightMode=RCTWM_WeightedNone,Time=0,Value=0,ArriveTangent=0,ArriveTangentWeight=0,LeaveTangent=0,LeaveTangentWeight=0),(InterpMode=RCIM_Cubic,TangentMode=RCTM_Auto,TangentWeightMo ...(truncated)"
    SourceMinimum=0
    TargetMinimum=0
    Value=@n3.Result
   )
   AlphaInterp = &RigVMFunction_AlphaInterp(InterpSpeedDecreasing=10, InterpSpeedIncreasing=10, Scale=1, bClampResult=false, Value=@n2.OutputParam)
   n5 = &RigVMFunction_AnimEvalRichCurve(
    Curve="(EditorCurveData=(Keys=((InterpMode=RCIM_Cubic,TangentMode=RCTM_Auto,TangentWeightMode=RCTWM_WeightedNone,Time=0,Value=0,ArriveTangent=0,ArriveTangentWeight=0,LeaveTangent=0,LeaveTangentWeight=0),(InterpMode=RCIM_Cubic,TangentMode=RCTM_Auto,TangentWeightMo ...(truncated)"
    SourceMinimum=0
    TargetMinimum=0
    Value=@AlphaInterp.Result
   )
   n6 = &RigUnit_GetTransform(Item=(Type=Bone,Name="spine_02_for_cons"), Space=LocalSpace, bInitial=false)
   RigVMFunction_MathQuaternionFromEuler = &RigVMFunction_MathQuaternionFromEuler(Euler=/Script/RigVM.RigVMFunction_MathVectorMul(
     B=(X=-1,Y=-1,Z=-1)
     B.X=&RigVMFunction_MathDoubleRemap(SourceMaximum=1, SourceMinimum=0, TargetMaximum=1, TargetMinimum=-1, Value=@n5.Result).Result
     A=&RigVMFunction_MathQuaternionToEuler(RotationOrder=ZYX, Value=@n6.Transform.Rotation).Result
    ) {is=struct RigVMUnitNode, template="Multiply::Execute(in A,in B,out Result)"}.Result)
   > Set Transform = &RigUnit_SetRotation(Item=(Type=Bone,Name="jacket_spine_cr"), bInitial=false, Weight=@n4.Result, Value=@RigVMFunction_MathQuaternionFromEuler.Result)
   RigVMFunctionReturnNode(ExecuteContext=@RigVMFunction_Sequence_1.D)
   n7 = &RigUnit_SphericalPoseReader(
    ActiveRegionScaleFactors=(PositiveWidth=1,NegativeWidth=1,PositiveHeight=1,NegativeHeight=1)
    ActiveRegionSize=0.35
    Debug=(bDrawDebug=false,bDraw2D=False,bDrawLocalAxes=False,DebugScale=25,DebugSegments=20,DebugThickness=0.25)
    DriverAxis=(X=-1,Y=0,Z=0)
    DriverItem=(Type=Bone,Name="thigh_l")
    FalloffRegionScaleFactors=(PositiveWidth=1,NegativeWidth=1,PositiveHeight=1,NegativeHeight=1)
    FalloffSize=0.55
    FlipHeightScaling=False
    FlipWidthScaling=False
    OptionalParentItem=(Type=Bone,Name="None")
    RotationOffset=(X=90,Y=0,Z=-20)
    ExecutePin=@RigVMFunction_Sequence_1.B
   )
   n8 = &RigVMFunction_MathVectorDistance(
    B=&RigUnit_GetTransform(Item=(Type=Bone,Name="calf_l"), bInitial=False).Transform.Translation
    A=&RigUnit_GetTransform(Item=(Type=Bone,Name="spine_03"), bInitial=False).Transform.Translation
   )
   > &RigUnit_ModifyTransforms(
    ItemToModify="((Item=(Type=Bone,Name=\"base_jacket_front_lf\"),Transform=(Rotation=(X=0.0038,Y=0.0871,Z=0.0435,W=-0.9952),Translation=(X=0,Y=0,Z=0),Scale3D=(X=1,Y=1,Z=1))),(Item=(Type=Bone,Name=\"base_jacket_side_front_lf\"),Transform=(Rotation=(X=0,Y=0,Z=0,W=1),Translation ...(truncated)"
    WeightMinimum=0
    Weight=&RigVMFunction_MathDoubleMul(
     B=&RigVMFunction_MathDoubleRemap(SourceMaximum=39.2818, SourceMinimum=63.8854, TargetMaximum=1, TargetMinimum=0, Value=@n8.Result).Result
     A=@n7.OutputParam
    ).Result
   )
   > n9 = &RigUnit_SphericalPoseReader(
    ActiveRegionScaleFactors=(PositiveWidth=1,NegativeWidth=1,PositiveHeight=1,NegativeHeight=1)
    ActiveRegionSize=0.05
    Debug=(bDrawDebug=false,bDraw2D=False,bDrawLocalAxes=False,DebugScale=25,DebugSegments=20,DebugThickness=0.25)
    DriverAxis=(X=-1,Y=0,Z=0)
    DriverItem=(Type=Bone,Name="thigh_l")
    FalloffRegionScaleFactors=(PositiveWidth=1,NegativeWidth=1,PositiveHeight=1,NegativeHeight=1)
    FalloffSize=0.4
    FlipHeightScaling=False
    FlipWidthScaling=False
    OptionalParentItem=(Type=Bone,Name="None")
    RotationOffset=(X=180,Y=20,Z=0)
   )
   > &RigUnit_ModifyTransforms(
    ItemToModify="((Item=(Type=Bone,Name=\"base_jacket_side_front_lf\"),Transform=(Rotation=(X=0,Y=-0.7071,Z=0,W=0.7071),Translation=(X=0,Y=0,Z=0),Scale3D=(X=1,Y=1,Z=1))),(Item=(Type=Bone,Name=\"base_jacket_side_back_lf\"),Transform=(Rotation=(X=0.0076,Y=-0.0868,Z=0.0868,W=0.99 ...(truncated)"
    WeightMinimum=0
    Weight=@n9.OutputParam
   )
   > n10 = &RigUnit_SphericalPoseReader(
    ActiveRegionScaleFactors=(PositiveWidth=1,NegativeWidth=1,PositiveHeight=1,NegativeHeight=1)
    ActiveRegionSize=0.12
    Debug=(bDrawDebug=false,bDraw2D=False,bDrawLocalAxes=False,DebugScale=25,DebugSegments=20,DebugThickness=0.25)
    DriverAxis=(X=-1,Y=0,Z=0)
    DriverItem=(Type=Bone,Name="thigh_l")
    FalloffRegionScaleFactors=(PositiveWidth=1,NegativeWidth=1,PositiveHeight=1,NegativeHeight=1)
    FalloffSize=0.35
    FlipHeightScaling=False
    FlipWidthScaling=False
    OptionalParentItem=(Type=Bone,Name="None")
    RotationOffset=(X=160,Y=30,Z=0)
   )
   > n11 = &RigUnit_ModifyTransforms(
    ItemToModify="((Item=(Type=Bone,Name=\"base_jacket_front_lf\"),Transform=(Rotation=(X=-0.1933,Y=0.613,Z=0.2304,W=-0.7306),Translation=(X=0,Y=0,Z=0),Scale3D=(X=1,Y=1,Z=1))),(Item=(Type=Bone,Name=\"base_jacket_side_front_lf\"),Transform=(Rotation=(X=0,Y=0,Z=0,W=1),Translation ...(truncated)"
    WeightMinimum=0
    Weight=@n10.OutputParam
   )
   Belt logic = RigVMCollapseNode(ExecutePin=@n11.ExecutePin, Weight=@n10.OutputParam)
   n12 = &RigUnit_SphericalPoseReader(
    ActiveRegionScaleFactors=(PositiveWidth=1,NegativeWidth=1,PositiveHeight=1,NegativeHeight=1)
    ActiveRegionSize=0.45
    Debug=(bDrawDebug=false,bDraw2D=False,bDrawLocalAxes=False,DebugScale=25,DebugSegments=20,DebugThickness=0.25)
    DriverAxis=(X=1,Y=0,Z=0)
    DriverItem=(Type=Bone,Name="thigh_r")
    FalloffRegionScaleFactors=(PositiveWidth=1,NegativeWidth=1,PositiveHeight=1,NegativeHeight=1)
    FalloffSize=0.55
    FlipHeightScaling=true
    FlipWidthScaling=true
    OptionalParentItem=(Type=Bone,Name="None")
    RotationOffset=(X=-90,Y=0,Z=-20)
    ExecutePin=@RigVMFunction_Sequence_1.C
   )
   RigUnit_GetTransform = &RigUnit_GetTransform(Item=(Type=Bone,Name="calf_r"), bInitial=False)
   RigVMFunction_MathVectorDistance = &RigVMFunction_MathVectorDistance(
    B=@RigUnit_GetTransform.Transform.Translation
    A=&RigUnit_GetTransform(Item=(Type=Bone,Name="spine_03"), bInitial=False).Transform.Translation
   )
   RigVMFunction_AnimEvalRichCurve = &RigVMFunction_AnimEvalRichCurve(
    Curve=(EditorCurveData=(Keys=((InterpMode=RCIM_Cubic,TangentMode=RCTM_User,ArriveTangent=2.9286,LeaveTangent=2.9286),(Time=1,Value=1))))
    SourceMinimum=0
    TargetMinimum=0
    Value=&RigVMFunction_MathDoubleRemap(SourceMaximum=39.2818, SourceMinimum=63.8854, TargetMaximum=1, TargetMinimum=0, Value=@RigVMFunction_MathVectorDistance.Result).Result
   )
   Multiply = &RigVMFunction_MathDoubleMul(B=@RigVMFunction_AnimEvalRichCurve.Result, A=@n12.OutputParam)
   > &RigUnit_ModifyTransforms(
    ItemToModify="((Item=(Type=Bone,Name=\"base_jacket_front_rt\"),Transform=(Rotation=(X=0,Y=0,Z=0.1305,W=0.9914),Translation=(X=0,Y=0,Z=0),Scale3D=(X=1,Y=1,Z=1))),(Item=(Type=Bone,Name=\"base_jacket_side_front_rt\"),Transform=(Rotation=(X=0.1861,Y=0.0067,Z=-0.0808,W=-0.9792), ...(truncated)"
    WeightMinimum=0
    Weight=@Multiply.Result
   )
   > n13 = &RigUnit_SphericalPoseReader(
    ActiveRegionScaleFactors=(PositiveWidth=1,NegativeWidth=1,PositiveHeight=1,NegativeHeight=1)
    ActiveRegionSize=0.05
    Debug=(bDrawDebug=false,bDraw2D=False,bDrawLocalAxes=False,DebugScale=25,DebugSegments=20,DebugThickness=0.25)
    DriverAxis=(X=1,Y=0,Z=0)
    DriverItem=(Type=Bone,Name="thigh_r")
    FalloffRegionScaleFactors=(PositiveWidth=1,NegativeWidth=1,PositiveHeight=1,NegativeHeight=1)
    FalloffSize=0.35
    FlipHeightScaling=true
    FlipWidthScaling=true
    OptionalParentItem=(Type=Bone,Name="None")
    RotationOffset=(X=0,Y=20,Z=0)
   )
   > &RigUnit_ModifyTransforms(
    ItemToModify="((Item=(Type=Bone,Name=\"base_jacket_side_front_rt\"),Transform=(Rotation=(X=0,Y=0.6428,Z=0,W=-0.766),Translation=(X=0,Y=0,Z=0),Scale3D=(X=1,Y=1,Z=1))),(Item=(Type=Bone,Name=\"base_jacket_front_rt\"),Transform=(Rotation=(X=0.0298,Y=-0.3407,Z=0.3971,W=0.8517),T ...(truncated)"
    WeightMinimum=0
    Weight=&RigVMFunction_MathDoubleMul(B=@RigVMFunction_AnimEvalRichCurve.Result, A=@n13.OutputParam).Result
   )
   > n14 = &RigUnit_SphericalPoseReader(
    ActiveRegionScaleFactors=(PositiveWidth=1,NegativeWidth=1,PositiveHeight=1,NegativeHeight=1)
    ActiveRegionSize=0.05
    Debug=(bDrawDebug=false,bDraw2D=False,bDrawLocalAxes=False,DebugScale=25,DebugSegments=20,DebugThickness=0.25)
    DriverAxis=(X=1,Y=0,Z=0)
    DriverItem=(Type=Bone,Name="thigh_r")
    FalloffRegionScaleFactors=(PositiveWidth=1,NegativeWidth=1,PositiveHeight=0.7,NegativeHeight=0.7)
    FalloffSize=0.35
    FlipHeightScaling=true
    FlipWidthScaling=true
    OptionalParentItem=(Type=Bone,Name="None")
    RotationOffset=(X=-40,Y=20,Z=0)
   )
   n15 = &RigVMFunction_MathDoubleMul(B=@RigVMFunction_AnimEvalRichCurve.Result, A=@n14.OutputParam)
   > &RigUnit_ModifyTransforms(
    ItemToModify="((Item=(Type=Bone,Name=\"base_jacket_front_rt\"),Transform=(Rotation=(X=0,Y=-0.0872,Z=0.0868,W=0.9924),Translation=(X=0,Y=0,Z=0),Scale3D=(X=1,Y=1,Z=1))),(Item=(Type=Bone,Name=\"base_jacket_side_front_rt\"),Transform=(Rotation=(X=0,Y=0,Z=0,W=1),Translation=(X=0 ...(truncated)"
    WeightMinimum=0
    Weight=@n15.Result
   )
   RigVMDispatchNode(Enabled=True, ExecuteContext=(), Prefix="FrontSide: : ", ScreenColor=(R=1,G=0.1944,B=0.9207,A=1), ScreenDuration=0.05, Value=@n15.Result) {template="DISPATCH_RigVMDispatch_Print(in Prefix,in Value,in Enabled,in ScreenDuration,in ScreenColor)"}
  ]}
 New Function_ContainedGraph {nodes [
   n1 = RigVMFunctionEntryNode()
   RigUnit_GetTransform = &RigUnit_GetTransform(Item=(Type=Bone,Name="thigh_l_for_cons"), Space=LocalSpace, bInitial=false, Item.Name=@n1.input joint)
   RigVMFunction_MathQuaternionToEuler = &RigVMFunction_MathQuaternionToEuler(
    RotationOrder=@n1.Rotation Order
    Value=&RigVMFunction_MathQuaternionSwingTwist(
     TwistAxis=(X=1,Y=0,Z=0)
     TwistAxis.X=@n1.Twist Axis.X
     TwistAxis.Y=@n1.Twist Axis.Y
     TwistAxis.Z=@n1.Twist Axis.Z
     Input=@RigUnit_GetTransform.Transform.Rotation
    ).Swing
   )
   n2 = &RigVMFunction_MathDoubleMul(B=@RigVMFunction_MathQuaternionToEuler.Result.Z, A=@n1.Axis to Get.Z)
   n3 = /Script/RigVM.RigVMFunction_MathDoubleLess(B=0, A=@n2.Result) {is=struct RigVMUnitNode, template="Less::Execute(in A,in B,out Result)"}
   n4 = /Script/RigVM.RigVMFunction_MathDoubleGreater(B=0, A=@n2.Result) {is=struct RigVMUnitNode, template="Greater::Execute(in A,in B,out Result)"}
   Add = &RigVMFunction_MathDoubleAdd(
    A=RigVMDispatchNode(
     False=0
     Condition=@n1.Forward
     True=RigVMDispatchNode(False=0, Condition=@n3.Result, True=@RigVMFunction_MathQuaternionToEuler.Result.Z) {template="DISPATCH_RigVMDispatch_If(in Condition,in True,in False,out Result)"}.Result
    ) {template="DISPATCH_RigVMDispatch_If(in Condition,in True,in False,out Result)"}.Result
    B=RigVMDispatchNode(
     False=0
     Condition=@n1.Backward
     True=RigVMDispatchNode(False=0, Condition=@n4.Result, True=@RigVMFunction_MathQuaternionToEuler.Result.Z) {template="DISPATCH_RigVMDispatch_If(in Condition,in True,in False,out Result)"}.Result
    ) {template="DISPATCH_RigVMDispatch_If(in Condition,in True,in False,out Result)"}.Result
   )
   n5 = &RigVMFunction_MathDoubleMul(A=@n1.Axis to Get.Y, B=@RigVMFunction_MathQuaternionToEuler.Result.Y)
   Less = &RigVMFunction_MathDoubleLess(B=0, A=@n5.Result) {template="Less::Execute(in A,in B,out Result)"}
   n6 = RigVMDispatchNode(False=0, Condition=@Less.Result, True=@RigVMFunction_MathQuaternionToEuler.Result.Y) {template="DISPATCH_RigVMDispatch_If(in Condition,in True,in False,out Result)"}
   Greater = &RigVMFunction_MathDoubleGreater(B=0, A=@n5.Result) {template="Greater::Execute(in A,in B,out Result)"}
   n7 = RigVMDispatchNode(False=0, Condition=@Greater.Result, True=@RigVMFunction_MathQuaternionToEuler.Result.Y) {template="DISPATCH_RigVMDispatch_If(in Condition,in True,in False,out Result)"}
   RigVMFunction_MathQuaternionFromEuler = &RigVMFunction_MathQuaternionFromEuler(
    Euler=(X=0,Y=0,Z=0)
    Euler.Z=@Add.Result
    Euler.Y=&RigVMFunction_MathDoubleAdd(
     A=RigVMDispatchNode(False=0, Condition=@n1.Outward, True=@n6.Result) {template="DISPATCH_RigVMDispatch_If(in Condition,in True,in False,out Result)"}.Result
     B=RigVMDispatchNode(False=0, Condition=@n1.Inward, True=@n7.Result) {template="DISPATCH_RigVMDispatch_If(in Condition,in True,in False,out Result)"}.Result
    ).Result
    RotationOrder=@n1.Rotation Order
   )
   > RigVMFunctionReturnNode(Output Rotate=(Value=(X=0,Y=0,Z=0,W=1)), Output Rotate.Value=@RigVMFunction_MathQuaternionFromEuler.Result)
  ]}
 New Function_ContainedGraph {nodes [
   n1 = RigVMFunctionEntryNode()
   RigUnit_GetTransform = &RigUnit_GetTransform(Item=(Type=Bone,Name="thigh_l_for_cons"), Space=LocalSpace, bInitial=false, Item.Name=@n1.input joint)
   RigVMFunction_MathQuaternionToEuler = &RigVMFunction_MathQuaternionToEuler(
    RotationOrder=@n1.Rotation Order
    Value=&RigVMFunction_MathQuaternionSwingTwist(
     TwistAxis=(X=1,Y=0,Z=0)
     TwistAxis.X=@n1.Twist Axis.X
     TwistAxis.Y=@n1.Twist Axis.Y
     TwistAxis.Z=@n1.Twist Axis.Z
     Input=@RigUnit_GetTransform.Transform.Rotation
    ).Swing
   )
   n2 = &RigVMFunction_MathDoubleMul(B=@RigVMFunction_MathQuaternionToEuler.Result.Z, A=@n1.Axis to Get.Z)
   n3 = &RigVMFunction_MathDoubleLess(B=0, A=@n2.Result) {template="Less::Execute(in A,in B,out Result)"}
   n4 = &RigVMFunction_MathDoubleGreater(B=0, A=@n2.Result) {template="Greater::Execute(in A,in B,out Result)"}
   Add = &RigVMFunction_MathDoubleAdd(
    A=RigVMDispatchNode(
     False=0
     Condition=@n1.Forward
     True=RigVMDispatchNode(False=0, Condition=@n3.Result, True=@RigVMFunction_MathQuaternionToEuler.Result.Z) {template="DISPATCH_RigVMDispatch_If(in Condition,in True,in False,out Result)"}.Result
    ) {template="DISPATCH_RigVMDispatch_If(in Condition,in True,in False,out Result)"}.Result
    B=RigVMDispatchNode(
     False=0
     Condition=@n1.Backward
     True=RigVMDispatchNode(False=0, Condition=@n4.Result, True=@RigVMFunction_MathQuaternionToEuler.Result.Z) {template="DISPATCH_RigVMDispatch_If(in Condition,in True,in False,out Result)"}.Result
    ) {template="DISPATCH_RigVMDispatch_If(in Condition,in True,in False,out Result)"}.Result
   )
   n5 = &RigVMFunction_MathDoubleMul(A=@n1.Axis to Get.Y, B=@RigVMFunction_MathQuaternionToEuler.Result.Y)
   Less = &RigVMFunction_MathDoubleLess(B=0, A=@n5.Result) {template="Less::Execute(in A,in B,out Result)"}
   n6 = RigVMDispatchNode(False=0, Condition=@Less.Result, True=@RigVMFunction_MathQuaternionToEuler.Result.Y) {template="DISPATCH_RigVMDispatch_If(in Condition,in True,in False,out Result)"}
   Greater = &RigVMFunction_MathDoubleGreater(B=0, A=@n5.Result) {template="Greater::Execute(in A,in B,out Result)"}
   n7 = RigVMDispatchNode(False=0, Condition=@Greater.Result, True=@RigVMFunction_MathQuaternionToEuler.Result.Y) {template="DISPATCH_RigVMDispatch_If(in Condition,in True,in False,out Result)"}
   RigVMFunction_MathQuaternionFromEuler = &RigVMFunction_MathQuaternionFromEuler(
    Euler=(X=0,Y=0,Z=0)
    Euler.Z=@Add.Result
    Euler.Y=&RigVMFunction_MathDoubleAdd(
     A=RigVMDispatchNode(False=0, Condition=@n1.Outward, True=@n6.Result) {template="DISPATCH_RigVMDispatch_If(in Condition,in True,in False,out Result)"}.Result
     B=RigVMDispatchNode(False=0, Condition=@n1.Inward, True=@n7.Result) {template="DISPATCH_RigVMDispatch_If(in Condition,in True,in False,out Result)"}.Result
    ).Result
    RotationOrder=@n1.Rotation Order
   )
   > RigVMFunctionReturnNode(Output Rotate=(Value=(X=0,Y=0,Z=0,W=1)), Output Rotate.Value=@RigVMFunction_MathQuaternionFromEuler.Result)
  ]}
 New Function_ContainedGraph {nodes [
   RigVMFunctionEntryNode()
   > &RigUnit_ParentConstraint(
    AdvancedSettings=(RotationOrderForFilter=XZY)
    Child=(Type=Bone,Name="shoe_l")
    Parents=((Item=(Type=Bone,Name="calf_l"),Weight=0.0144),(Item=(Type=Bone,Name="calf_twist_02_l"),Weight=0.0351),(Item=(Type=Bone,Name="calf_twist_01_l"),Weight=0.3441),(Item=(Type=Bone,Name="foot_l"),Weight=0.6064))
    Weight=1
   )
   > Delta Rotation_3 = RigVMFunctionReferenceNode(Child=foot_l, Parent=calf_l, Source Maximum=-50, Source Minimum=0)
   > &RigUnit_ModifyTransforms(
    ItemToModify=((Item=(Type=Bone,Name="shoe_l"),Transform=(Rotation=(X=0,Y=0,Z=0.7071,W=0.7071),Translation=(X=0,Y=0,Z=0),Scale3D=(X=1,Y=1,Z=1))))
    WeightMinimum=0
    Weight=@"Delta Rotation_3".delta
   )
   > &RigUnit_ParentConstraint(
    AdvancedSettings=(RotationOrderForFilter=XZY)
    Child=(Type=Bone,Name="base_shoe_r")
    Parents=((Item=(Type=Bone,Name="calf_r"),Weight=0.0047),(Item=(Type=Bone,Name="calf_twist_02_r"),Weight=0.0108),(Item=(Type=Bone,Name="calf_twist_01_r"),Weight=0.4062),(Item=(Type=Bone,Name="foot_r"),Weight=0.5783))
    Weight=1
   )
   > Delta Rotation_1 = RigVMFunctionReferenceNode(Child=foot_r, Parent=calf_r, Source Maximum=-50, Source Minimum=0)
   > &RigUnit_ModifyTransforms(
    ItemToModify=((Item=(Type=Bone,Name="shoe_r"),Transform=(Rotation=(X=0,Y=0,Z=0.7071,W=0.7071),Translation=(X=0,Y=0,Z=0),Scale3D=(X=1,Y=1,Z=1))))
    WeightMinimum=0
    Weight=@"Delta Rotation_1".delta
   )
   > RigVMFunctionReturnNode()
  ]}
 New Function_ContainedGraph {nodes [
   n1 = RigVMFunctionEntryNode()
   RigVMFunction_MathQuaternionToEuler = &RigVMFunction_MathQuaternionToEuler(
    RotationOrder=ZYX
    Value=&RigUnit_GetRelativeTransformForItem(
     Child=(Type=Bone,Name="hand_l")
     Parent=(Type=Bone,Name="lowerarm_l")
     bChildInitial=False
     bParentInitial=False
     Child.Name=@n1.Child
     Parent.Name=@n1.Parent
    ).RelativeTransform.Rotation
   )
   Remap = &RigVMFunction_MathDoubleRemap(
    TargetMaximum=1
    TargetMinimum=0
    Value=@RigVMFunction_MathQuaternionToEuler.Result.Z
    SourceMaximum=@n1.Source Maximum
    SourceMinimum=@n1.Source Minimum
   )
   > RigVMFunctionReturnNode(delta=@Remap.Result)
  ]}
 New Function_ContainedGraph {nodes [
   n1 = RigVMFunctionEntryNode()
   > n2 = &RigUnit_SphericalPoseReader(
    ActiveRegionScaleFactors=(PositiveWidth=1,NegativeWidth=1,PositiveHeight=1,NegativeHeight=1)
    ActiveRegionSize=0.05
    Debug=(bDrawDebug=false,bDraw2D=False,bDrawLocalAxes=False,DebugScale=25,DebugSegments=20,DebugThickness=0.25)
    DriverAxis=(X=-1,Y=0,Z=0)
    DriverItem=(Type=Bone,Name="thigh_l")
    FalloffRegionScaleFactors=(PositiveWidth=1,NegativeWidth=1,PositiveHeight=1,NegativeHeight=1)
    FalloffSize=0.55
    FlipHeightScaling=False
    FlipWidthScaling=False
    OptionalParentItem=(Type=Bone,Name="None")
    RotationOffset=(X=90,Y=0,Z=-20)
   )
   > n3 = &RigUnit_SphericalPoseReader(
    ActiveRegionScaleFactors=(PositiveWidth=1,NegativeWidth=1,PositiveHeight=1,NegativeHeight=1)
    ActiveRegionSize=0.05
    Debug=(bDrawDebug=false,bDraw2D=False,bDrawLocalAxes=False,DebugScale=25,DebugSegments=20,DebugThickness=0.25)
    DriverAxis=(X=-1,Y=0,Z=0)
    DriverItem=(Type=Bone,Name="thigh_l")
    FalloffRegionScaleFactors=(PositiveWidth=1,NegativeWidth=1,PositiveHeight=1,NegativeHeight=1)
    FalloffSize=0.55
    FlipHeightScaling=False
    FlipWidthScaling=False
    OptionalParentItem=(Type=Bone,Name="None")
    RotationOffset=(X=140,Y=30,Z=0)
   )
   > n4 = &RigUnit_SphericalPoseReader(
    ActiveRegionScaleFactors=(PositiveWidth=1,NegativeWidth=1,PositiveHeight=1,NegativeHeight=1)
    ActiveRegionSize=0.05
    Debug=(bDrawDebug=false,bDraw2D=False,bDrawLocalAxes=False,DebugScale=25,DebugSegments=20,DebugThickness=0.25)
    DriverAxis=(X=-1,Y=0,Z=0)
    DriverItem=(Type=Bone,Name="thigh_l")
    FalloffRegionScaleFactors=(PositiveWidth=1,NegativeWidth=1,PositiveHeight=1,NegativeHeight=1)
    FalloffSize=0.55
    FlipHeightScaling=False
    FlipWidthScaling=False
    OptionalParentItem=(Type=Bone,Name="None")
    RotationOffset=(X=180,Y=20,Z=0)
   )
   > n5 = &RigUnit_SphericalPoseReader(
    ActiveRegionScaleFactors=(PositiveWidth=1,NegativeWidth=1,PositiveHeight=1,NegativeHeight=1)
    ActiveRegionSize=0.05
    Debug=(bDrawDebug=false,bDraw2D=False,bDrawLocalAxes=False,DebugScale=25,DebugSegments=20,DebugThickness=0.25)
    DriverAxis=(X=-1,Y=0,Z=0)
    DriverItem=(Type=Bone,Name="thigh_l")
    FalloffRegionScaleFactors=(PositiveWidth=1,NegativeWidth=1,PositiveHeight=1,NegativeHeight=1)
    FalloffSize=0.55
    FlipHeightScaling=False
    FlipWidthScaling=False
    OptionalParentItem=(Type=Bone,Name="None")
    RotationOffset=(X=-140,Y=30,Z=0)
   )
   > n6 = &RigUnit_SphericalPoseReader(
    ActiveRegionScaleFactors=(PositiveWidth=1,NegativeWidth=1,PositiveHeight=1,NegativeHeight=1)
    ActiveRegionSize=0.05
    Debug=(bDrawDebug=false,bDraw2D=False,bDrawLocalAxes=False,DebugScale=25,DebugSegments=20,DebugThickness=0.25)
    DriverAxis=(X=-1,Y=0,Z=0)
    DriverItem=(Type=Bone,Name="thigh_l")
    FalloffRegionScaleFactors=(PositiveWidth=1,NegativeWidth=1,PositiveHeight=1,NegativeHeight=1)
    FalloffSize=0.55
    FlipHeightScaling=False
    FlipWidthScaling=False
    OptionalParentItem=(Type=Bone,Name="None")
    RotationOffset=(X=-90,Y=0,Z=20)
   )
   Multiply = &RigVMFunction_MathDoubleMul(A=@n2.OutputParam, B=@n1.delta)
   > RigVMFunctionReturnNode(
    Front=@Multiply.Result
    FrontMid=&RigVMFunction_MathDoubleMul(A=@n3.OutputParam, B=@n1.delta).Result
    Mid=&RigVMFunction_MathDoubleMul(A=@n4.OutputParam, B=@n1.delta).Result
    Argument=&RigVMFunction_MathDoubleMul(A=@n5.OutputParam, B=@n1.delta).Result
    Back=&RigVMFunction_MathDoubleMul(A=@n6.OutputParam, B=@n1.delta).Result
   )
  ]}
 New Function_ContainedGraph {nodes [
   n1 = RigVMFunctionEntryNode()
   > n2 = &RigUnit_SphericalPoseReader(
    ActiveRegionScaleFactors=(PositiveWidth=1,NegativeWidth=1,PositiveHeight=1,NegativeHeight=1)
    ActiveRegionSize=0.05
    Debug=(bDrawDebug=false,bDraw2D=False,bDrawLocalAxes=False,DebugScale=25,DebugSegments=20,DebugThickness=0.25)
    DriverAxis=(X=1,Y=0,Z=0)
    DriverItem=(Type=Bone,Name="thigh_r")
    FalloffRegionScaleFactors=(PositiveWidth=1,NegativeWidth=1,PositiveHeight=1,NegativeHeight=1)
    FalloffSize=0.55
    FlipHeightScaling=true
    FlipWidthScaling=true
    OptionalParentItem=(Type=Bone,Name="None")
    RotationOffset=(X=-90,Y=0,Z=-20)
   )
   > n3 = &RigUnit_SphericalPoseReader(
    ActiveRegionScaleFactors=(PositiveWidth=1,NegativeWidth=1,PositiveHeight=1,NegativeHeight=1)
    ActiveRegionSize=0.05
    Debug=(bDrawDebug=false,bDraw2D=False,bDrawLocalAxes=False,DebugScale=25,DebugSegments=20,DebugThickness=0.25)
    DriverAxis=(X=1,Y=0,Z=0)
    DriverItem=(Type=Bone,Name="thigh_r")
    FalloffRegionScaleFactors=(PositiveWidth=1,NegativeWidth=1,PositiveHeight=1,NegativeHeight=1)
    FalloffSize=0.55
    FlipHeightScaling=true
    FlipWidthScaling=true
    OptionalParentItem=(Type=Bone,Name="None")
    RotationOffset=(X=-40,Y=20,Z=0)
   )
   > n4 = &RigUnit_SphericalPoseReader(
    ActiveRegionScaleFactors=(PositiveWidth=1,NegativeWidth=1,PositiveHeight=1,NegativeHeight=1)
    ActiveRegionSize=0.05
    Debug=(bDrawDebug=false,bDraw2D=False,bDrawLocalAxes=False,DebugScale=25,DebugSegments=20,DebugThickness=0.25)
    DriverAxis=(X=1,Y=0,Z=0)
    DriverItem=(Type=Bone,Name="thigh_r")
    FalloffRegionScaleFactors=(PositiveWidth=1,NegativeWidth=1,PositiveHeight=1,NegativeHeight=1)
    FalloffSize=0.55
    FlipHeightScaling=true
    FlipWidthScaling=true
    OptionalParentItem=(Type=Bone,Name="None")
    RotationOffset=(X=0,Y=20,Z=0)
   )
   > n5 = &RigUnit_SphericalPoseReader(
    ActiveRegionScaleFactors=(PositiveWidth=1,NegativeWidth=1,PositiveHeight=1,NegativeHeight=1)
    ActiveRegionSize=0.05
    Debug=(bDrawDebug=false,bDraw2D=False,bDrawLocalAxes=False,DebugScale=25,DebugSegments=20,DebugThickness=0.25)
    DriverAxis=(X=1,Y=0,Z=0)
    DriverItem=(Type=Bone,Name="thigh_r")
    FalloffRegionScaleFactors=(PositiveWidth=1,NegativeWidth=1,PositiveHeight=1,NegativeHeight=1)
    FalloffSize=0.55
    FlipHeightScaling=true
    FlipWidthScaling=true
    OptionalParentItem=(Type=Bone,Name="None")
    RotationOffset=(X=40,Y=20,Z=0)
   )
   > n6 = &RigUnit_SphericalPoseReader(
    ActiveRegionScaleFactors=(PositiveWidth=1,NegativeWidth=1,PositiveHeight=1,NegativeHeight=1)
    ActiveRegionSize=0.05
    Debug=(bDrawDebug=false,bDraw2D=False,bDrawLocalAxes=False,DebugScale=25,DebugSegments=20,DebugThickness=0.25)
    DriverAxis=(X=1,Y=0,Z=0)
    DriverItem=(Type=Bone,Name="thigh_r")
    FalloffRegionScaleFactors=(PositiveWidth=1,NegativeWidth=1,PositiveHeight=1,NegativeHeight=1)
    FalloffSize=0.55
    FlipHeightScaling=true
    FlipWidthScaling=true
    OptionalParentItem=(Type=Bone,Name="None")
    RotationOffset=(X=90,Y=0,Z=20)
   )
   Multiply = &RigVMFunction_MathDoubleMul(A=@n2.OutputParam, B=@n1.Delta)
   > RigVMFunctionReturnNode(
    Front=@Multiply.Result
    FrontSide=&RigVMFunction_MathDoubleMul(A=@n3.OutputParam, B=@n1.Delta).Result
    Side=&RigVMFunction_MathDoubleMul(A=@n4.OutputParam, B=@n1.Delta).Result
    BackSide=&RigVMFunction_MathDoubleMul(A=@n5.OutputParam, B=@n1.Delta).Result
    Back=&RigVMFunction_MathDoubleMul(A=@n6.OutputParam, B=@n1.Delta).Result
   )
  ]}
 RigVMFunctionLibrary {nodes [
   Delta Rotation = RigVMCollapseNode(ExecuteContext=())
   Jacket = RigVMCollapseNode(ExecuteContext=())
   Jacket_Left = RigVMCollapseNode(ExecuteContext=())
   Jacket_Right = RigVMCollapseNode(ExecuteContext=())
   shoes = RigVMCollapseNode(ExecuteContext=())
  ]}
 RigVMFunction_Sequence_2_ContainedGraph {nodes [
   n1 = RigVMFunctionEntryNode()
   n2 = &RigVMFunction_Sequence(ExecuteContext=@n1.ExecuteContext)
   n3 = &RigVMFunction_Sequence(ExecuteContext=@n2.B)
   n4 = &RigVMFunction_Sequence(ExecuteContext=@n3.B)
   RigVMFunctionReturnNode(A=@n2.A, B=@n3.A, C=@n4.A, D=@n4.B)
  ]}
 RigVMModel {nodes [
   RigUnit_BeginExecution = &RigUnit_BeginExecution()
   RigVMFunction_Sequence = &RigVMFunction_Sequence(ExecuteContext=@RigUnit_BeginExecution.ExecutePin)
   FN_DeformRig_FNC = RigVMFunctionReferenceNode(
    DeformRig_ Curve="(deformRig_l_upperarm_up_alpha_72_#1=False,deformRig_r_upperarm_up_alpha_73_#2=False,deformRig_l_upperarm_fr_alpha_74_#3=False,deformRig_r_upperarm_fr_alpha_75_CB51B8 ...(truncated)"
    DeformRig_Bool="(upperarm_1_#4=True,elbow_3_#5=True,wrist_7_#6=True,groin_8_#7=True,glute_17_#8=True,knee_9_40C408324CA261 ...(truncated)"
    DeformRig_Debug=(upperarm_debug_18_#4=False,wrist_debug_20_#6=False,lat_debug_26_#9=False,groinGlute_debug_27_#7=False)
    DeformRig_Value="(upperarm_up_18_#10=(Rotation=(X=0.0,Y=-0.0,Z=0.0,W=1.0),Translation=(X=0.0,Y=0.0,Z=8.0),Scale3D=(X=0.70,Y=1.0,Z=1.0)),upperarm_fwd_fr_19_#11=(Rotation=(X=0.0,Y=-0.0,Z=0.0,W=1.0),Translation=(X=0.0, ...(truncated)"
    ExecuteContext=@RigVMFunction_Sequence.A
   )
   > &RigUnit_ParentConstraint(
    AdvancedSettings=(RotationOrderForFilter=XZY)
    Child=(Type=Bone,Name="base_lowerarm_belt_rt")
    Parents=((Item=(Type=Bone,Name="lowerarm_r"),Weight=0.2185),(Item=(Type=Bone,Name="lowerarm_twist_02_r"),Weight=0.2817),(Item=(Type=Bone,Name="lowerarm_twist_01_r"),Weight=0.4997))
    Weight=1
   )
   > &RigUnit_ParentConstraint(
    AdvancedSettings=(RotationOrderForFilter=XZY)
    Child=(Type=Bone,Name="base_lowerarm_belt_lf")
    Parents=((Item=(Type=Bone,Name="lowerarm_l"),Weight=0.2181),(Item=(Type=Bone,Name="lowerarm_twist_02_l"),Weight=0.2927),(Item=(Type=Bone,Name="lowerarm_twist_01_l"),Weight=0.4893))
    Weight=1
   )
   Jacket = RigVMFunctionReferenceNode(ExecuteContext=@RigVMFunction_Sequence.B)
   > &RigUnit_ParentConstraint(
    AdvancedSettings=(RotationOrderForFilter=XZY)
    Child=(Type=Bone,Name="aim_belt_lf")
    Parents="((Item=(Type=Bone,Name=\"deform_groin_l\"),Weight=0.0105),(Item=(Type=Bone,Name=\"deform_glute_l\"),Weight=0.0268),(Item=(Type=Bone,Name=\"pelvis\"),Weight=0.1222),(Item=(Type=Bone,Name=\"thigh_l\"),Weight=0.1542),(Item=(Type=Bone,Name=\"thigh_twist_01_l\"),Weight=0 ...(truncated)"
    Weight=1
   )
   > &RigUnit_AimConstraintLocalSpaceOffset(
    AdvancedSettings=(DebugSettings=(bEnabled=true,Scale=10,WorldOffset=(Rotation=(X=0,Y=0,Z=0,W=1),Translation=(X=0,Y=0,Z=0),Scale3D=(X=1,Y=1,Z=1))),RotationOrderForFilter=XZY)
    Child=(Type=Bone,Name="offset_belt_lf")
    Filter=(bX=true,bY=true,bZ=true)
    Parents=((Item=(Type=Bone,Name="aim_belt_lf"),Weight=1))
    Weight=1
    WorldUp=(Target=(X=0,Y=0,Z=1),Kind=Direction,Space=(Type=Bone,Name="offset_belt_lf"))
   )
   > shoes = RigVMFunctionReferenceNode()
   > n1 = &RigUnit_SphericalPoseReader(
    ActiveRegionScaleFactors=(PositiveWidth=1,NegativeWidth=1,PositiveHeight=1,NegativeHeight=1)
    ActiveRegionSize=0.07
    Debug=(bDrawDebug=false,bDraw2D=False,bDrawLocalAxes=False,DebugScale=25,DebugSegments=20,DebugThickness=0.25)
    DriverAxis=(X=-1,Y=0,Z=0)
    DriverItem=(Type=Bone,Name="upperarm_r")
    FalloffRegionScaleFactors=(PositiveWidth=1,NegativeWidth=1,PositiveHeight=1,NegativeHeight=1)
    FalloffSize=0.15
    FlipHeightScaling=False
    FlipWidthScaling=False
    OptionalParentItem=(Type=Bone,Name="None")
    RotationOffset=(X=-40,Y=30,Z=0)
   )
   > &RigUnit_ModifyTransforms(
    ItemToModify=((Item=(Type=Bone,Name="base_jacket_front_rt"),Transform=(Rotation=(X=0,Y=-0.0436,Z=0,W=0.999),Translation=(X=-0.202,Y=0,Z=-2.3092),Scale3D=(X=1,Y=1,Z=1))))
    WeightMinimum=0
    Weight=@n1.OutputParam
   )
  ]}
]
hierarchy [
 root {loc=<0 0 0>, rot=<0 0 0 1>}
 pelvis {parent=root, loc=<0 -0.5301 98.0571>, rot=<0.0127 0.707 -0.0127 -0.707>}
 spine_01 {parent=pelvis, loc=<6.1208 -0.9328 0>, rot=<0 0 0.034 -0.9994>}
 spine_02 {parent=spine_01, loc=<6.5997 -0.2095 0>, rot=<0 0 0.0238 -0.9997>}
 spine_03 {parent=spine_02, loc=<6.4914 -0.1237 0>, rot=<0 0 0.0149 -0.9999>}
 spine_04 {parent=spine_03, loc=<6.9598 0.0794 0>, rot=<0 0 -0.0132 -0.9999>}
 spine_05 {parent=spine_04, loc=<6.9987 0 0>, rot=<0 0 -0.0015 -1>}
 neck_01 {parent=spine_05, loc=<11.3138 0.1529 0>, rot=<0 0 -0.2003 -0.9797>}
 neck_02 {parent=neck_01, loc=<4.5627 0 0>, rot=<0 0 0.0504 -0.9987>}
 head {parent=neck_02, loc=<5.3914 0 0>, rot=<0 0 0.115 -0.9934>}
 clavicle_l {parent=spine_05, loc=<5.4831 -0.1053 -3.2946>, rot=<0.0465 -0.7289 -0.009 -0.6829>}
 upperarm_l {parent=clavicle_l, loc=<9.8658 -1.2136 -0.4105>, rot=<0.0164 -0.3245 -0.0269 -0.9454>}
 lowerarm_l {parent=upperarm_l, loc=<26.465 -0.8345 0.259>, rot=<0.0036 0.0161 -0.2298 -0.9731>}
 hand_l {parent=lowerarm_l, loc=<22.7355 -0.3203 -0.6105>, rot=<-0.6784 -0.0223 0.0735 -0.7307>}
 thumb_01_l {parent=hand_l, loc=<1.3118 -0.3191 -2.459>, rot=<0.6546 -0.4042 0.0664 -0.6354>}
 thumb_02_l {parent=thumb_01_l, loc=<5.9336 -0.0448 0.0401>, rot=<-0.0037 0.0213 0.298 -0.9543>}
 thumb_03_l {parent=thumb_02_l, loc=<2.2223 0.3682 -0.0907>, rot=<0.0118 -0.013 0.0066 -0.9998>}
 index_metacarpal_l {parent=hand_l, loc=<1.6535 0.0183 -2.96>, rot=<0.0158 -0.0124 -0.0173 -0.9996>}
 index_01_l {parent=index_metacarpal_l, loc=<8.3008 0.0357 -0.0524>, rot=<-0.0044 0.0485 0.0924 -0.9945>}
 index_02_l {parent=index_01_l, loc=<3.3587 -0.3004 0.0452>, rot=<-0.002 0.0027 0.3225 -0.9466>}
 index_03_l {parent=index_02_l, loc=<2.5578 -0.123 -0.0252>, rot=<-0.002 0.0067 0.2558 -0.9667>}
 middle_metacarpal_l {parent=hand_l, loc=<2.1856 -0.1725 -0.5984>, rot=<0.006 0.0168 -0.0329 -0.9993>}
 middle_01_l {parent=middle_metacarpal_l, loc=<7.3275 -0.2447 0.0129>, rot=<-0.0208 0.0042 0.0858 -0.9961>}
 middle_02_l {parent=middle_01_l, loc=<4.0123 0.1401 0.1175>, rot=<-0.0218 0.0258 0.3512 -0.9357>}
 middle_03_l {parent=middle_02_l, loc=<2.6077 0.0534 0.1373>, rot=<0.0082 0.0188 0.1708 -0.9851>}
 ring_metacarpal_l {parent=hand_l, loc=<2.2074 -0.3625 1.3638>, rot=<-0.0284 0.0722 -0.0458 -0.9959>}
 ring_01_l {parent=ring_metacarpal_l, loc=<7.0295 -0.6197 -0.1446>, rot=<-0.0489 -0.0252 0.1892 -0.9804>}
 ring_02_l {parent=ring_01_l, loc=<3.6174 0.4647 -0.1027>, rot=<0 0 0.2806 -0.9598>}
 ring_03_l {parent=ring_02_l, loc=<2.4152 0.114 0.0199>, rot=<0 0 0.1189 -0.9929>}
 pinky_metacarpal_l {parent=hand_l, loc=<1.7616 -0.9753 3.3814>, rot=<-0.0397 0.0814 -0.069 -0.9935>}
 pinky_01_l {parent=pinky_metacarpal_l, loc=<7.1709 -0.4831 -0.1094>, rot=<-0.0654 -0.0035 0.2242 -0.9723>}
 pinky_02_l {parent=pinky_01_l, loc=<2.8274 0.0893 -0.1043>, rot=<0.0037 0.0011 0.2854 -0.9584>}
 pinky_03_l {parent=pinky_02_l, loc=<2.2046 0.0307 -0.0078>, rot=<-0.0002 0.0098 0.1445 -0.9895>}
 deform_wrist_dn_base_l {parent=hand_l, loc=<0 -1.5326 0>, rot=<0 0 0 -1>}
 deform_wrist_dn_l {parent=deform_wrist_dn_base_l, loc=<0 0 0>, rot=<0 0 0 1>}
 deform_wrist_up_base_l {parent=hand_l, loc=<0 2.3009 0>, rot=<0 0 0 -1>}
 deform_wrist_up_l {parent=deform_wrist_up_base_l, loc=<0 0 0>, rot=<0 0 0 1>}
 lowerarm_twist_02_l {parent=lowerarm_l, loc=<10.6707 0 0>, rot=<0 0 0 1>}
 lowerarm_twist_01_l {parent=lowerarm_l, loc=<18.316 0 0>, rot=<0 0 0 1>}
 deform_elbow_in_base_l {parent=lowerarm_l, loc=<0.1834 3.6062 -0.2572>, rot=<0 0 0 -1>}
 deform_elbow_in_l {parent=deform_elbow_in_base_l, loc=<0 0 0>, rot=<0 0 0 1>}
 deform_elbow_out_base_l {parent=lowerarm_l, loc=<0 -4 0>, rot=<0 0 0 1>}
 deform_elbow_out_l {parent=deform_elbow_out_base_l, loc=<0 0 0>, rot=<0 0 0 1>}
 base_lowerarm_belt_lf {parent=lowerarm_l, loc=<11.3358 -2.483 3.9409>, rot=<0.2521 0.0922 0.6642 -0.6977>}
 lowerarm_belt_lf {parent=base_lowerarm_belt_lf, loc=<0 0 0>, rot=<0 0 0 1>}
 upperarm_twist_01_l {parent=upperarm_l, loc=<7.2135 -0.1134 0>, rot=<0 0 0 1>}
 deform_upperarm_bk_base_l {parent=upperarm_twist_01_l, loc=<-6.4053 -4.3746 0.314>, rot=<0 0 0 -1>}
 deform_upperarm_bk_l {parent=deform_upperarm_bk_base_l, loc=<0 0 0>, rot=<0 0 0 1>}
 deform_upperarm_fr_base_l {parent=upperarm_twist_01_l, loc=<-5.6312 5.092 0.3354>, rot=<0 0 0 -1>}
 deform_upperarm_fr_l {parent=deform_upperarm_fr_base_l, loc=<0 0 0>, rot=<0 0 0 1>}
 deform_upperarm_up_base_l {parent=upperarm_twist_01_l, loc=<-8.891 0.4825 3.9517>, rot=<0 0 0 -1>}
 deform_upperarm_up_l {parent=deform_upperarm_up_base_l, loc=<0 0 0>, rot=<0 0 0 1>}
 base_straps_lf {parent=upperarm_twist_01_l, loc=<1.2805 -5.1201 3.4356>, rot=<0.4345 -0.0888 0.1122 -0.8892>}
 dyn_straps_01_lf {parent=base_straps_lf, loc=<0 0 0>, rot=<0 0 0 1>}
 dyn_straps_02_lf {parent=dyn_straps_01_lf, loc=<5.5466 0 0>, rot=<0 0 0 1>}
 dyn_straps_03_lf {parent=dyn_straps_02_lf, loc=<5.0018 0 0>, rot=<0 0 0 1>}
 aim_straps_lf {parent=upperarm_twist_01_l, loc=<16.8357 -9.6073 2.4568>, rot=<0.4345 -0.0888 0.1122 -0.8892>}
 upperarm_twist_02_l {parent=upperarm_l, loc=<19.7998 -0.1134 0>, rot=<0 0 0 1>}
 clavicle_r {parent=spine_05, loc=<5.4831 -0.1053 3.2946>, rot=<0.7289 0.0465 -0.6829 0.009>}
 upperarm_r {parent=clavicle_r, loc=<-9.8658 1.2136 0.4105>, rot=<0.0164 -0.3245 -0.0269 -0.9454>}
 lowerarm_r {parent=upperarm_r, loc=<-26.465 0.8345 -0.259>, rot=<0.0036 0.0161 -0.2298 -0.9731>}
 hand_r {parent=lowerarm_r, loc=<-22.7355 0.3203 0.6106>, rot=<-0.6784 -0.0223 0.0735 -0.7307>}
 thumb_01_r {parent=hand_r, loc=<-1.3118 0.3191 2.459>, rot=<0.6546 -0.4042 0.0664 -0.6354>}
 thumb_02_r {parent=thumb_01_r, loc=<-5.9336 0.0448 -0.0401>, rot=<-0.0037 0.0213 0.298 -0.9543>}
 thumb_03_r {parent=thumb_02_r, loc=<-2.2223 -0.3682 0.0907>, rot=<0.0118 -0.013 0.0066 -0.9998>}
 index_metacarpal_r {parent=hand_r, loc=<-1.6535 -0.0183 2.96>, rot=<0.0158 -0.0124 -0.0173 -0.9996>}
 index_01_r {parent=index_metacarpal_r, loc=<-8.3008 -0.0357 0.0524>, rot=<-0.0044 0.0485 0.0924 -0.9945>}
 index_02_r {parent=index_01_r, loc=<-3.3587 0.3004 -0.0452>, rot=<-0.002 0.0027 0.3225 -0.9466>}
 index_03_r {parent=index_02_r, loc=<-2.5578 0.123 0.0252>, rot=<-0.002 0.0067 0.2558 -0.9667>}
 middle_metacarpal_r {parent=hand_r, loc=<-2.1856 0.1725 0.5984>, rot=<0.006 0.0168 -0.0329 -0.9993>}
 middle_01_r {parent=middle_metacarpal_r, loc=<-7.3275 0.2447 -0.0129>, rot=<-0.0208 0.0042 0.0858 -0.9961>}
 middle_02_r {parent=middle_01_r, loc=<-4.0123 -0.1401 -0.1175>, rot=<-0.0218 0.0258 0.3512 -0.9357>}
 middle_03_r {parent=middle_02_r, loc=<-2.6077 -0.0534 -0.1373>, rot=<0.0082 0.0188 0.1708 -0.9851>}
 ring_metacarpal_r {parent=hand_r, loc=<-2.2074 0.3625 -1.3638>, rot=<-0.0284 0.0722 -0.0458 -0.9959>}
 ring_01_r {parent=ring_metacarpal_r, loc=<-7.0295 0.6197 0.1446>, rot=<-0.0489 -0.0252 0.1892 -0.9804>}
 ring_02_r {parent=ring_01_r, loc=<-3.6174 -0.4647 0.1027>, rot=<0 0 0.2806 -0.9598>}
 ring_03_r {parent=ring_02_r, loc=<-2.4152 -0.114 -0.0199>, rot=<0 0 0.1189 -0.9929>}
 pinky_metacarpal_r {parent=hand_r, loc=<-1.7616 0.9753 -3.3814>, rot=<-0.0397 0.0814 -0.069 -0.9935>}
 pinky_01_r {parent=pinky_metacarpal_r, loc=<-7.1709 0.4831 0.1094>, rot=<-0.0654 -0.0035 0.2242 -0.9723>}
 pinky_02_r {parent=pinky_01_r, loc=<-2.8274 -0.0893 0.1043>, rot=<0.0037 0.0011 0.2854 -0.9584>}
 pinky_03_r {parent=pinky_02_r, loc=<-2.2046 -0.0307 0.0078>, rot=<-0.0002 0.0098 0.1445 -0.9895>}
 deform_wrist_dn_base_r {parent=hand_r, loc=<-0.0001 1.5328 0>, rot=<0 0 0 -1>}
 deform_wrist_dn_r {parent=deform_wrist_dn_base_r, loc=<0 0 0>, rot=<0 0 0 1>}
 deform_wrist_up_base_r {parent=hand_r, loc=<-0.0002 -2.3007 -0.0001>, rot=<0 0 0 -1>}
 deform_wrist_up_r {parent=deform_wrist_up_base_r, loc=<0 0 0>, rot=<0 0 0 1>}
 lowerarm_twist_02_r {parent=lowerarm_r, loc=<-10.6707 0 0>, rot=<0 0 0 -1>}
 lowerarm_twist_01_r {parent=lowerarm_r, loc=<-18.316 0 0>, rot=<0 0 0 -1>}
 deform_elbow_in_base_r {parent=lowerarm_r, loc=<-0.1834 -3.6062 0.2572>, rot=<0 0 0 -1>}
 deform_elbow_in_r {parent=deform_elbow_in_base_r, loc=<0 0 0>, rot=<0 0 0 1>}
 deform_elbow_out_base_r {parent=lowerarm_r, loc=<0 4 -0.0001>, rot=<0 0 0 -1>}
 deform_elbow_out_r {parent=deform_elbow_out_base_r, loc=<0 0 0>, rot=<0 0 0 1>}
 base_lowerarm_belt_rt {parent=lowerarm_r, loc=<-11.3357 2.483 -3.941>, rot=<-0.6642 -0.6977 0.2521 -0.0922>}
 lowerarm_belt_rt {parent=base_lowerarm_belt_rt, loc=<0 0 0>, rot=<0 0 0 1>}
 upperarm_twist_01_r {parent=upperarm_r, loc=<-7.2135 0.1134 0>, rot=<0 0 0 -1>}
 deform_upperarm_bk_base_r {parent=upperarm_twist_01_r, loc=<6.4053 4.3746 -0.314>, rot=<0 0 0 -1>}
 deform_upperarm_bk_r {parent=deform_upperarm_bk_base_r, loc=<0 0 0>, rot=<0 0 0 1>}
 deform_upperarm_fr_base_r {parent=upperarm_twist_01_r, loc=<5.6312 -5.092 -0.3354>, rot=<0 0 0 -1>}
 deform_upperarm_fr_r {parent=deform_upperarm_fr_base_r, loc=<0 0 0>, rot=<0 0 0 1>}
 deform_upperarm_up_base_r {parent=upperarm_twist_01_r, loc=<8.891 -0.4825 -3.9517>, rot=<0 0 0 -1>}
 deform_upperarm_up_r {parent=deform_upperarm_up_base_r, loc=<0 0 0>, rot=<0 0 0 1>}
 aim_straps_rt {parent=upperarm_twist_01_r, loc=<-16.8359 9.6072 -2.4565>, rot=<0.0888 0.4345 0.8892 0.1122>}
 base_straps_rt {parent=upperarm_twist_01_r, loc=<-1.2803 5.1201 -3.436>, rot=<0.0888 0.4345 0.8892 0.1122>}
 dyn_straps_01_rt {parent=base_straps_rt, loc=<0 0 0>, rot=<0 0 0 -1>}
 dyn_straps_02_rt {parent=dyn_straps_01_rt, loc=<5.5469 0.0003 0.0001>, rot=<0 0 0 -1>}
 dyn_straps_03_rt {parent=dyn_straps_02_rt, loc=<5.0021 0.0003 0.0002>, rot=<0 0 0 -1>}
 upperarm_twist_02_r {parent=upperarm_r, loc=<-19.7998 0.1134 0>, rot=<0 0 0 -1>}
 attach_backpack {parent=spine_05, loc=<-0.123 -13.473 0>, rot=<-0.0093 -0.707 0.0093 -0.707>}
 Backpack_BR {parent=attach_backpack, type=Null, loc=<0 0 0>, rot=<0 0 0 1>}
 attach_cape {parent=spine_05, loc=<9.017 -11.714 0>, rot=<0.0616 -0.7044 -0.0616 -0.7044>}
 cape {parent=attach_cape, type=Null, loc=<0 -3.742 0>, rot=<0 0 0 1>}
 deform_pec_aim_base_l {parent=spine_04, loc=<3.2456 9.3309 -1.2237>, rot=<0.0603 -0.4033 0.3318 -0.8507>}
 deform_pec_aim_l {parent=deform_pec_aim_base_l, loc=<0 0 0>, rot=<0 0 0 1>}
 deform_pec_base_l {parent=deform_pec_aim_l, loc=<9.3665 0 0>, rot=<0 0 0 1>}
 deform_pec_l {parent=deform_pec_base_l, loc=<0 0 0>, rot=<0 0 0 1>}
 deform_pec_aim_base_r {parent=spine_04, loc=<3.2452 9.3309 1.2237>, rot=<-0.4033 -0.0603 0.8507 0.3318>}
 deform_pec_aim_r {parent=deform_pec_aim_base_r, loc=<0 0 0>, rot=<0 0 0 1>}
 deform_pec_base_r {parent=deform_pec_aim_r, loc=<-9.3667 -0.0003 -0.0004>, rot=<0 0 0 1>}
 deform_pec_r {parent=deform_pec_base_r, loc=<0 0 0>, rot=<0 0 0 1>}
 base_jacket_spine_cr {parent=spine_03, loc=<0 0 0>, rot=<0 0 0 1>}
 jacket_spine_cr {parent=base_jacket_spine_cr, loc=<0 0 0>, rot=<0 0 0 1>}
 base_jacket_front_lf {parent=jacket_spine_cr, loc=<0.4977 7.735 -12.4338>, rot=<0.1016 -0.9039 -0.4111 -0.0602>}
 dyn_jacket_front_01_lf {parent=base_jacket_front_lf, loc=<0 0 0>, rot=<0 0 -0.0132 -0.9999>}
 dyn_jacket_front_02_lf {parent=dyn_jacket_front_01_lf, loc=<12.844 0 0>, rot=<0 0 0 1>}
 dyn_jacket_front_03_lf {parent=dyn_jacket_front_02_lf, loc=<11.7132 0 0>, rot=<0 0 0 1>}
 base_jacket_side_front_lf {parent=jacket_spine_cr, loc=<1.3115 -2.951 -14.2516>, rot=<0.0686 -0.9798 0.1571 -0.1032>}
 dyn_jacket_side_front_01_lf {parent=base_jacket_side_front_lf, loc=<0 0 0>, rot=<0 0 -0.0132 -0.9999>}
 dyn_jacket_side_front_02_lf {parent=dyn_jacket_side_front_01_lf, loc=<12.8173 0 0>, rot=<0 0 0 1>}
 dyn_jacket_side_front_03_lf {parent=dyn_jacket_side_front_02_lf, loc=<9.9845 0 0>, rot=<0 0 0 1>}
 base_jacket_side_back_lf {parent=jacket_spine_cr, loc=<2.1408 -9.2916 -9.3483>, rot=<0.1204 -0.7965 0.5868 -0.0822>}
 dyn_jacket_side_back_01_lf {parent=base_jacket_side_back_lf, loc=<0 0 0>, rot=<0 0 -0.0132 -0.9999>}
 dyn_jacket_side_back_02_lf {parent=dyn_jacket_side_back_01_lf, loc=<11.2563 0 0>, rot=<0 0 0 1>}
 dyn_jacket_side_back_03_lf {parent=dyn_jacket_side_back_02_lf, loc=<8.0087 0 0>, rot=<0 0 0 1>}
 base_jacket_back_cr {parent=jacket_spine_cr, loc=<1.7379 -10.7201 0>, rot=<0.1538 -0.6902 0.6902 -0.1538>}
 dyn_jacket_back_01_cr {parent=base_jacket_back_cr, loc=<0 0 0>, rot=<0 0 -0.0132 -0.9999>}
 dyn_jacket_back_02_cr {parent=dyn_jacket_back_01_cr, loc=<9.1218 0 0>, rot=<0 0 0 1>}
 dyn_jacket_back_03_cr {parent=dyn_jacket_back_02_cr, loc=<6.9209 0 0>, rot=<0 0 0 1>}
 base_jacket_front_rt {parent=jacket_spine_cr, loc=<0.4977 7.735 12.4338>, rot=<0.0602 0.4111 0.9039 -0.1016>}
 dyn_jacket_front_01_rt {parent=base_jacket_front_rt, loc=<0 0 0>, rot=<0 0 0.0132 -0.9999>}
 dyn_jacket_front_02_rt {parent=dyn_jacket_front_01_rt, loc=<12.8436 0.0001 0.0001>, rot=<0 0 0 1>}
 dyn_jacket_front_03_rt {parent=dyn_jacket_front_02_rt, loc=<11.7136 -0.0001 -0.0001>, rot=<0 0 0 1>}
 base_jacket_side_front_rt {parent=jacket_spine_cr, loc=<1.3116 -2.951 14.2516>, rot=<0.1032 -0.1571 0.9798 -0.0686>}
 dyn_jacket_side_front_01_rt {parent=base_jacket_side_front_rt, loc=<0 0 0>, rot=<0 0 0.0132 -0.9999>}
 dyn_jacket_side_front_02_rt {parent=dyn_jacket_side_front_01_rt, loc=<12.8171 0 0>, rot=<0 0 0 1>}
 dyn_jacket_side_front_03_rt {parent=dyn_jacket_side_front_02_rt, loc=<9.9848 0 -0.0001>, rot=<0 0 0 1>}
 base_jacket_side_back_rt {parent=jacket_spine_cr, loc=<2.1407 -9.2916 9.3483>, rot=<0.0822 -0.5868 0.7965 -0.1204>}
 dyn_jacket_side_back_01_rt {parent=base_jacket_side_back_rt, loc=<0 0 0>, rot=<0 0 0.0132 -0.9999>}
 dyn_jacket_side_back_02_rt {parent=dyn_jacket_side_back_01_rt, loc=<11.2562 0 0>, rot=<0 0 0 1>}
 dyn_jacket_side_back_03_rt {parent=dyn_jacket_side_back_02_rt, loc=<8.0087 0 0>, rot=<0 0 0 1>}
 spine_03_for_cons_offset {parent=spine_03, loc=<0 0 0>, rot=<0 0 0 1>}
 spine_03_for_cons {parent=spine_03_for_cons_offset, loc=<0 0 0>, rot=<0 0 0 1>}
 spine_02_for_cons_offset {parent=spine_01, loc=<6.5988 -0.2111 0.0003>, rot=<0 0 0.0238 -0.9997>}
 spine_02_for_cons {parent=spine_02_for_cons_offset, loc=<0 0 0>, rot=<0 0 0 1>}
 thigh_l {parent=pelvis, loc=<-1.0855 -0.2109 -8.7053>, rot=<0.0009 0.0355 0.0155 -0.9993>}
 calf_l {parent=thigh_l, loc=<-42.5284 -0.0019 -0.7297>, rot=<0.0006 -0.0161 -0.0391 -0.9991>}
 foot_l {parent=calf_l, loc=<-46.7401 -2.5014 -0.592>, rot=<0.0427 -0.0238 0.0353 -0.9982>}
 ball_l {parent=foot_l, loc=<-3.5658 10.8066 -0.1544>, rot=<0.0001 -0.0001 -0.7063 -0.7079>}
 base_shoe_l {parent=foot_l, loc=<6.5972 7.642 -0.1539>, rot=<0.0001 -0.0001 -0.7063 -0.7079>}
 shoe_l {parent=base_shoe_l, loc=<0 0 0>, rot=<0 0 0 1>}
 calf_twist_02_l {parent=calf_l, loc=<-19.1546 -1.7943 -0.2638>, rot=<0.0237 0.0095 -0.0035 -0.9997>}
 calf_twist_01_l {parent=calf_l, loc=<-35.8483 -1.9197 -0.5794>, rot=<0.0237 0.0095 -0.0035 -0.9997>}
 deform_knee_in_base_l {parent=calf_l, loc=<0.8031 -5.0231 0.2456>, rot=<0 0 0 -1>}
 deform_knee_in_l {parent=deform_knee_in_base_l, loc=<0 0 0>, rot=<0 0 0 1>}
 deform_knee_out_base_l {parent=calf_l, loc=<0 4 0>, rot=<0 0 0 -1>}
 deform_knee_out_l {parent=deform_knee_out_base_l, loc=<0 0 0>, rot=<0 0 0 1>}
 thigh_twist_01_l {parent=thigh_l, loc=<-27.952 0 0>, rot=<0.0013 0 0 -1>}
 deform_groin_base_l {parent=thigh_l, loc=<0.4801 5.6776 -0.2486>, rot=<0 0 0 -1>}
 deform_groin_l {parent=deform_groin_base_l, loc=<0 0 0>, rot=<0 0 0 1>}
 thigh_r {parent=pelvis, loc=<-1.0855 -0.2109 8.7053>, rot=<0.0355 -0.0009 0.9993 0.0155>}
 calf_r {parent=thigh_r, loc=<42.5284 0.0019 0.7297>, rot=<0.0006 -0.0161 -0.0391 -0.9991>}
 foot_r {parent=calf_r, loc=<46.7401 2.5014 0.592>, rot=<0.0427 -0.0238 0.0353 -0.9982>}
 ball_r {parent=foot_r, loc=<3.5658 -10.8066 0.1544>, rot=<0.0001 -0.0001 -0.7063 -0.7079>}
 base_shoe_r {parent=foot_r, loc=<-6.6345 -7.6097 0.1536>, rot=<0.0001 -0.001 0.7097 -0.7045>}
 shoe_r {parent=base_shoe_r, loc=<0 0 0>, rot=<0 0 0 1>}
 calf_twist_02_r {parent=calf_r, loc=<19.1546 1.7943 0.2638>, rot=<0.0237 0.0095 -0.0035 -0.9997>}
 calf_twist_01_r {parent=calf_r, loc=<35.8483 1.9197 0.5794>, rot=<0.0237 0.0095 -0.0035 -0.9997>}
 deform_knee_in_base_r {parent=calf_r, loc=<-0.8031 5.0231 -0.2457>, rot=<0 0 0 -1>}
 deform_knee_in_r {parent=deform_knee_in_base_r, loc=<0 0 0>, rot=<0 0 0 1>}
 deform_knee_out_base_r {parent=calf_r, loc=<0 -4 0>, rot=<0 0 0 -1>}
 deform_knee_out_r {parent=deform_knee_out_base_r, loc=<0 0 0>, rot=<0 0 0 1>}
 thigh_twist_01_r {parent=thigh_r, loc=<27.952 0 0>, rot=<0.0013 0 0 -1>}
 deform_groin_base_r {parent=thigh_r, loc=<-0.4801 -5.6776 0.2486>, rot=<0 0 0 -1>}
 deform_groin_r {parent=deform_groin_base_r, loc=<0 0 0>, rot=<0 0 0 1>}
 deform_glute_base_l {parent=pelvis, loc=<-7.3466 -9.0213 -9.1245>, rot=<0.0172 0.031 -0.4477 -0.8935>}
 deform_glute_l {parent=deform_glute_base_l, loc=<0 0 0>, rot=<0 0 0 1>}
 deform_glute_base_r {parent=pelvis, loc=<-7.3464 -9.0213 9.1245>, rot=<0.031 -0.0172 0.8935 -0.4477>}
 deform_glute_r {parent=deform_glute_base_r, loc=<0 0 0>, rot=<0 0 0 1>}
 base_strap_back_lf {parent=pelvis, loc=<1.0412 -12.8117 -8.2007>, rot=<0.0127 0.707 0.707 0.0127>}
 offset_strap_back_lf {parent=base_strap_back_lf, loc=<0 0 0>, rot=<0 0 0 1>}
 dyn_strap_back_01_lf {parent=offset_strap_back_lf, loc=<0 0 0>, rot=<-0.0045 -0.0578 0.1463 -0.9875>}
 dyn_strap_back_02_lf {parent=dyn_strap_back_01_lf, loc=<6.4176 0 0>, rot=<0 0.136 0 -0.9907>}
 dyn_strap_back_03_lf {parent=dyn_strap_back_02_lf, loc=<6.4176 0 0>, rot=<0 0.07 0 -0.9976>}
 dyn_strap_back_04_lf {parent=dyn_strap_back_03_lf, loc=<6.4176 0 0>, rot=<-0.0001 0.0851 0.0201 -0.9962>}
 dyn_strap_back_05_lf {parent=dyn_strap_back_04_lf, loc=<6.4176 0 0>, rot=<0 0.0683 0 -0.9977>}
 dyn_strap_back_06_lf {parent=dyn_strap_back_05_lf, loc=<6.4176 0 0>, rot=<0 0.0094 0 -1>}
 base_belt_lf {parent=pelvis, loc=<7.3745 6.8327 -7.9162>, rot=<0.0127 0.707 0.707 0.0127>}
 belt_lf {parent=base_belt_lf, loc=<0 0 0>, rot=<0 0 0 1>}
 base_strap_front_lf {parent=belt_lf, loc=<5.6255 -0.6868 0.3871>, rot=<0 0 0 1>}
 offset_strap_front_lf {parent=base_strap_front_lf, loc=<0 0 0>, rot=<0 0 0 1>}
 dyn_strap_front_01_lf {parent=offset_strap_front_lf, loc=<0 0 0>, rot=<-0.0006 -0.0034 0.0904 -0.9959>}
 dyn_strap_front_02_lf {parent=dyn_strap_front_01_lf, loc=<6.5534 0 0>, rot=<0 -0.0011 0.0195 -0.9998>}
 dyn_strap_front_03_lf {parent=dyn_strap_front_02_lf, loc=<5.9788 0 0>, rot=<0.0069 -0.005 0.0328 -0.9994>}
 dyn_strap_front_04_lf {parent=dyn_strap_front_03_lf, loc=<5.9788 0 0>, rot=<-0.0036 -0.0958 0.035 -0.9948>}
 dyn_strap_front_05_lf {parent=dyn_strap_front_04_lf, loc=<5.9788 0 0>, rot=<0 -0.1238 0 -0.9923>}
 dyn_strap_front_06_lf {parent=dyn_strap_front_05_lf, loc=<5.9788 0 0>, rot=<0 0 0 1>}
 base_strap_front_second_lf {parent=belt_lf, loc=<6.1768 -0.6469 -0.7496>, rot=<0 0 0 1>}
 offset_strap_front_second_lf {parent=base_strap_front_second_lf, loc=<0 0 0>, rot=<-0.241 0.0695 0.2247 -0.9416>}
 strap_front_second_01_lf {parent=offset_strap_front_second_lf, loc=<0 0 0>, rot=<-0.0059 -0.0127 0.093 -0.9956>}
 strap_front_second_02_lf {parent=strap_front_second_01_lf, loc=<5.1207 0 0>, rot=<-0.0226 -0.1066 0.2706 -0.9565>}
 strap_front_second_03_lf {parent=strap_front_second_02_lf, loc=<4.138 0 0>, rot=<0 0 0 1>}
 thigh_l_get_rot_off {parent=pelvis, loc=<-1.0855 -0.2109 -8.7053>, rot=<0.0009 0.0355 0.0155 -0.9993>}
 thigh_l_get_rot {parent=thigh_l_get_rot_off, loc=<0 0 0>, rot=<0 0 0 1>}
 thigh_l_set_rot_front_off {parent=pelvis, loc=<-1.0855 -0.2109 -8.7053>, rot=<0.0009 0.0355 0.0155 -0.9992>}
 thigh_l_set_rot_front {parent=thigh_l_set_rot_front_off, loc=<0 0 0>, rot=<0 0 0 1>}
 thigh_l_set_rot_back_off {parent=pelvis, loc=<-1.0855 -0.2109 -8.7053>, rot=<0.0009 0.0355 0.0155 -0.9992>}
 thigh_l_set_rot_back {parent=thigh_l_set_rot_back_off, loc=<0 0 0>, rot=<0 0 0 1>}
 base_belt_lf1 {parent=pelvis, loc=<3.5717 -0.9701 -13.2264>, rot=<-0.0164 -0.9862 -0.0049 -0.1645>}
 offset_belt_lf {parent=base_belt_lf1, loc=<0 0 0>, rot=<0 0 0 1>}
 dyn_belt_lf {parent=offset_belt_lf, loc=<0 0 0>, rot=<0 0 0 1>}
 aim_belt_lf {parent=pelvis, loc=<-8.5892 -1.5448 -16.3617>, rot=<-0.0179 -0.9972 0.0013 -0.073>}
 Tail {parent=pelvis, type=Null, loc=<1.197 -10.3 -0.01>, rot=<-0.1753 0.6901 0.1524 0.6854>}
]
variables [Is PickAxe Gauntlet {type=bool, default=False}]
