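# Physics asset for the F_MED_BoldTouch female body, giving its jacket and left-side straps capsule bodies for secondary motion.
# Pelvis, spine_03, both thighs and the right calf provide body shapes.
# Six three-link jacket chains hang from free-rotating base bodies on spine_03, with tight angular limits that mostly widen toward front hems and allow almost no twist.
# Front and back straps attach to the pelvis with free rotation.
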PhysicsAsset /BRCosmetics/Characters/Player/Female/Medium/Bodies/F_MED_BoldTouch/Meshes/F_MED_BoldTouch_Physics
bodies {
 base_jacket_back_cr [{shapeType=Capsule, length=0.505, radius=0.505, shapeName=base_jacket_back_cr_capsule, rotation=<rot 0 -90 0>}]
 base_jacket_front_lf [{shapeType=Capsule, length=8, radius=2, shapeName=base_jacket_front_lf_capsule, center=<5 0 0>, rotation=<rot 0 90 90>}]
 base_jacket_front_rt [{shapeType=Capsule, length=0.505, radius=0.505, shapeName=base_jacket_front_rt_capsule, rotation=<rot 0 -90 0>}]
 base_jacket_side_back_lf [{shapeType=Capsule, length=0.505, radius=0.505, shapeName=base_jacket_side_back_lf_capsule, rotation=<rot 0 -90 0>}]
 base_jacket_side_back_rt [{shapeType=Capsule, length=0.505, radius=0.505, shapeName=base_jacket_side_back_rt_capsule, rotation=<rot 0 -90 0>}]
 base_jacket_side_front_lf [{shapeType=Capsule, length=0.505, radius=0.505, shapeName=base_jacket_side_front_lf_capsule, rotation=<rot 0 -90 0>}]
 base_jacket_side_front_rt [{shapeType=Capsule, length=0.505, radius=0.505, shapeName=base_jacket_side_front_rt_capsule, rotation=<rot 0 -90 0>}]
 calf_r [{
   shapeType=Capsule
   length=36.2654
   radius=9.0753
   shapeName=calf_r_capsule
   center=<24.9012 0.9299 0.0967>
   rotation=<rot -0.1284 97.3859 -89.6472>
  }]
 dyn_jacket_back_01_cr [{shapeType=Capsule, length=8, radius=2, shapeName=dyn_jacket_back_01_cr_capsule, center=<5 0 0>, rotation=<rot 0 90 90>}]
 dyn_jacket_back_02_cr [{shapeType=Capsule, length=8, radius=2, shapeName=dyn_jacket_back_01_cr_capsule, center=<5 0 0>, rotation=<rot 0 0 90>}]
 dyn_jacket_back_03_cr [{shapeType=Capsule, length=4.1345, radius=7.8655, shapeName=dyn_jacket_back_03_cr_capsule, center=<5 0 0>, rotation=<rot 0 0 90>}]
 dyn_jacket_front_01_lf [{shapeType=Capsule, length=7.361, radius=2.639, shapeName=dyn_jacket_front_01_lf_capsule, center=<5 0 0>, rotation=<rot 0 90 90>}]
 dyn_jacket_front_01_rt [{shapeType=Capsule, length=8, radius=2, shapeName=dyn_jacket_front_01_rt_capsule, center=<5 0 0>, rotation=<rot 0 90 90>}]
 dyn_jacket_front_02_lf [{shapeType=Capsule, length=8.0893, radius=3.4107, shapeName=dyn_jacket_front_02_rt_capsule, center=<5 0 0>, rotation=<rot 0 -90 89.9999>}]
 dyn_jacket_front_02_rt [{shapeType=Capsule, length=8, radius=3.5, shapeName=dyn_jacket_front_02_rt_capsule, center=<5 3.0557 -0.1979>, rotation=<rot 0 0 120>}]
 dyn_jacket_front_03_lf [
  {shapeType=Capsule, length=8.9779, radius=3.0221, shapeName=dyn_jacket_front_03_lf_capsule, center=<-2.2025 0 0>, rotation=<rot 90 0 0>}
  {
   shapeType=Capsule
   length=13.7216
   radius=1.9704
   shapeName=dyn_jacket_front_03_lf_capsule
   center=<0 -4.3889 -2.0466>
   rotation=<rot 0.0002 0.0001 65>
  }
 ]
 dyn_jacket_front_03_rt [{
   shapeType=Capsule
   length=9.1376
   radius=3.3624
   shapeName=dyn_jacket_front_03_rt_capsule
   center=<2.0379 -0.0744 0.0521>
   rotation=<rot 2.8658 -4.0994 124.8975>
  }]
 dyn_jacket_side_back_01_lf [{shapeType=Capsule, length=8, radius=2, center=<5 0 0>, rotation=<rot 0 90 90>}]
 dyn_jacket_side_back_01_rt [{shapeType=Capsule, length=8, radius=2, center=<5 0 0>, rotation=<rot 0 90 90>}]
 dyn_jacket_side_back_02_lf [{shapeType=Capsule, length=8, radius=2, shapeName=dyn_jacket_side_back_02_lf_capsule, center=<5 0 0>, rotation=<rot 0 0 90>}]
 dyn_jacket_side_back_02_rt [{shapeType=Capsule, length=8, radius=2, shapeName=dyn_jacket_side_back_02_rt_capsule, center=<5 0 0>, rotation=<rot 0 0 90>}]
 dyn_jacket_side_back_03_lf [{shapeType=Capsule, length=7.5569, radius=4.4431, shapeName=dyn_jacket_side_back_03_lf_capsule, center=<5 0 0>, rotation=<rot 0 0 90>}]
 dyn_jacket_side_back_03_rt [{shapeType=Capsule, length=7.0209, radius=4.9791, shapeName=dyn_jacket_side_back_03_rt_capsule, center=<5 0 0>, rotation=<rot 0 0 90>}]
 dyn_jacket_side_front_01_lf [{shapeType=Capsule, length=8, radius=2, shapeName=dyn_jacket_side_front_01_lf_capsule, center=<5 0 0>, rotation=<rot 0 90 90>}]
 dyn_jacket_side_front_01_rt [{shapeType=Capsule, length=8, radius=2, shapeName=dyn_jacket_side_front_01_rt_capsule, center=<5 0 0>, rotation=<rot 0 90 90>}]
 dyn_jacket_side_front_02_lf [{shapeType=Capsule, length=8.4102, radius=3.0898, shapeName=dyn_jacket_side_front_02_lf_capsule, center=<5 0 0>, rotation=<rot 0 0 90>}]
 dyn_jacket_side_front_02_rt [{shapeType=Capsule, length=8, radius=2, shapeName=dyn_jacket_side_front_02_rt_capsule, center=<5 0 0>, rotation=<rot 0 0 90>}]
 dyn_jacket_side_front_03_lf [{shapeType=Capsule, length=15.8818, radius=3.8297, shapeName=dyn_jacket_side_front_03_lf_capsule, center=<5 0 0>, rotation=<rot 0 0 90>}]
 dyn_jacket_side_front_03_rt [{shapeType=Capsule, length=11.5, radius=3.5134, shapeName=dyn_jacket_side_front_03_rt_capsule, center=<5 0 0>, rotation=<rot 0 0 90>}]
 dyn_strap_back_02_lf [{
   shapeType=Capsule
   length=3.9366
   radius=2.2523
   shapeName=dyn_strap_back_02_lf_capsule
   center=<2.2728 0.2987 0.2069>
   rotation=<rot 0 -91.5452 -86.1895>
  }]
 dyn_strap_front_01_lf [{shapeType=Capsule, length=5.51, radius=1.49, shapeName=dyn_strap_front_01_lf_capsule, center=<5 0 0>, rotation=<rot 90 0 0>}]
 dyn_strap_front_02_lf [{shapeType=Capsule, length=5.6966, radius=1.3034, shapeName=dyn_strap_front_02_lf_capsule, center=<5 0 0>, rotation=<rot 90 0 0>}]
 pelvis [{shapeType=Capsule, length=10.5, radius=9.2316, center=<-1.7014 -2.18 -0.5465>}]
 spine_03 [{shapeType=Capsule, length=16.7125, radius=10.7331, center=<-13.1371 -2.748 -0.0968>, rotation=<rot 0 -90 -90>}]
 thigh_l [{shapeType=Capsule, length=23.2559, radius=8.9996, center=<-22.1941 -2.1107 0.5123>, rotation=<rot 0 -89.3611 -95.5722>}]
 thigh_r [{shapeType=Capsule, length=23.2559, radius=8.9996, center=<22.1941 2.1107 -0.5123>, rotation=<rot 0 -89.3611 84.4278>}]
}
constraints [
 {bone1Name=spine_03, bone2Name=pelvis, swing1LimitDegrees=45, swing2LimitDegrees=45, twistLimitDegrees=45}
 {bone1Name=dyn_jacket_side_back_01_lf, bone2Name=base_jacket_side_back_lf, swing1LimitDegrees=0, swing2LimitDegrees=5, twistLimitDegrees=0}
 {bone1Name=dyn_jacket_side_back_01_rt, bone2Name=base_jacket_side_back_rt, swing1LimitDegrees=0, swing2LimitDegrees=5, twistLimitDegrees=0}
 {bone1Name=thigh_l, bone2Name=pelvis, swing1LimitDegrees=45, swing2LimitDegrees=45, twistLimitDegrees=45}
 {bone1Name=thigh_r, bone2Name=pelvis, swing1LimitDegrees=45, swing2LimitDegrees=45, twistLimitDegrees=45}
 {
  bone1Name=base_jacket_back_cr
  bone2Name=spine_03
  swing1LimitDegrees=45
  swing1Motion=Free
  swing2LimitDegrees=45
  swing2Motion=Free
  twistLimitDegrees=45
  twistMotion=Free
 }
 {bone1Name=dyn_jacket_back_01_cr, bone2Name=base_jacket_back_cr, swing1LimitDegrees=0, swing2LimitDegrees=5, twistLimitDegrees=0}
 {bone1Name=dyn_jacket_back_02_cr, bone2Name=dyn_jacket_back_01_cr, swing1LimitDegrees=0, swing2LimitDegrees=5, twistLimitDegrees=0}
 {
  bone1Name=base_jacket_front_lf
  bone2Name=spine_03
  swing1LimitDegrees=45
  swing1Motion=Free
  swing2LimitDegrees=45
  swing2Motion=Free
  twistLimitDegrees=45
  twistMotion=Free
 }
 {bone1Name=dyn_jacket_front_01_lf, bone2Name=base_jacket_front_lf, swing1LimitDegrees=0, swing2LimitDegrees=5, twistLimitDegrees=0}
 {bone1Name=dyn_jacket_front_02_lf, bone2Name=dyn_jacket_front_01_lf, swing1LimitDegrees=10, swing2LimitDegrees=10, twistLimitDegrees=0}
 {
  bone1Name=base_jacket_front_rt
  bone2Name=spine_03
  swing1LimitDegrees=45
  swing1Motion=Free
  swing2LimitDegrees=45
  swing2Motion=Free
  twistLimitDegrees=45
  twistMotion=Free
 }
 {bone1Name=dyn_jacket_front_01_rt, bone2Name=base_jacket_front_rt, swing1LimitDegrees=0, swing2LimitDegrees=5, twistLimitDegrees=0}
 {bone1Name=dyn_jacket_front_02_rt, bone2Name=dyn_jacket_front_01_rt, swing1LimitDegrees=10, swing2LimitDegrees=10, twistLimitDegrees=10}
 {
  bone1Name=base_jacket_side_back_lf
  bone2Name=spine_03
  swing1LimitDegrees=45
  swing1Motion=Free
  swing2LimitDegrees=45
  swing2Motion=Free
  twistLimitDegrees=45
  twistMotion=Free
 }
 {bone1Name=dyn_jacket_side_back_02_lf, bone2Name=dyn_jacket_side_back_01_lf, swing1LimitDegrees=0, swing2LimitDegrees=5, twistLimitDegrees=0}
 {
  bone1Name=base_jacket_side_back_rt
  bone2Name=spine_03
  swing1LimitDegrees=45
  swing1Motion=Free
  swing2LimitDegrees=45
  swing2Motion=Free
  twistLimitDegrees=45
  twistMotion=Free
 }
 {bone1Name=dyn_jacket_side_back_02_rt, bone2Name=dyn_jacket_side_back_01_rt, swing1LimitDegrees=0, swing2LimitDegrees=5, twistLimitDegrees=0}
 {
  bone1Name=base_jacket_side_front_lf
  bone2Name=spine_03
  swing1LimitDegrees=45
  swing1Motion=Free
  swing2LimitDegrees=45
  swing2Motion=Free
  twistLimitDegrees=45
  twistMotion=Free
 }
 {bone1Name=dyn_jacket_side_front_01_lf, bone2Name=base_jacket_side_front_lf, swing1LimitDegrees=0, swing2LimitDegrees=5, twistLimitDegrees=0}
 {
  bone1Name=dyn_jacket_side_front_02_lf
  bone2Name=dyn_jacket_side_front_01_lf
  swing1LimitDegrees=0
  swing2LimitDegrees=10
  twistLimitDegrees=0
 }
 {
  bone1Name=base_jacket_side_front_rt
  bone2Name=spine_03
  swing1LimitDegrees=45
  swing1Motion=Free
  swing2LimitDegrees=45
  swing2Motion=Free
  twistLimitDegrees=45
  twistMotion=Free
 }
 {bone1Name=dyn_jacket_side_front_01_rt, bone2Name=base_jacket_side_front_rt, swing1LimitDegrees=0, swing2LimitDegrees=5, twistLimitDegrees=0}
 {
  bone1Name=dyn_jacket_side_front_02_rt
  bone2Name=dyn_jacket_side_front_01_rt
  swing1LimitDegrees=0
  swing2LimitDegrees=10
  twistLimitDegrees=0
 }
 {bone1Name=dyn_jacket_back_03_cr, bone2Name=dyn_jacket_back_02_cr, swing1LimitDegrees=0, swing2LimitDegrees=5, twistLimitDegrees=0}
 {bone1Name=dyn_jacket_front_03_lf, bone2Name=dyn_jacket_front_02_lf, swing1LimitDegrees=15, swing2LimitDegrees=15, twistLimitDegrees=0}
 {bone1Name=dyn_jacket_front_03_rt, bone2Name=dyn_jacket_front_02_rt, swing1LimitDegrees=20, swing2LimitDegrees=15, twistLimitDegrees=0}
 {bone1Name=dyn_jacket_side_back_03_lf, bone2Name=dyn_jacket_side_back_02_lf, swing1LimitDegrees=0, swing2LimitDegrees=5, twistLimitDegrees=0}
 {bone1Name=dyn_jacket_side_back_03_rt, bone2Name=dyn_jacket_side_back_02_rt, swing1LimitDegrees=0, swing2LimitDegrees=10, twistLimitDegrees=0}
 {
  bone1Name=dyn_jacket_side_front_03_lf
  bone2Name=dyn_jacket_side_front_02_lf
  swing1LimitDegrees=0
  swing2LimitDegrees=15
  twistLimitDegrees=0
 }
 {
  bone1Name=dyn_jacket_side_front_03_rt
  bone2Name=dyn_jacket_side_front_02_rt
  swing1LimitDegrees=0
  swing2LimitDegrees=15
  twistLimitDegrees=0
 }
 {
  bone1Name=dyn_strap_front_01_lf
  bone2Name=pelvis
  swing1LimitDegrees=45
  swing1Motion=Free
  swing2LimitDegrees=45
  swing2Motion=Free
  twistLimitDegrees=45
  twistMotion=Free
 }
 {
  bone1Name=dyn_strap_front_02_lf
  bone2Name=dyn_strap_front_01_lf
  swing1LimitDegrees=45
  swing1Motion=Free
  swing2LimitDegrees=45
  swing2Motion=Free
  twistLimitDegrees=45
  twistMotion=Free
 }
 {
  bone1Name=calf_r
  bone2Name=thigh_r
  swing1LimitDegrees=45
  swing1Motion=Free
  swing2LimitDegrees=45
  swing2Motion=Free
  twistLimitDegrees=45
  twistMotion=Free
 }
 {
  bone1Name=dyn_strap_back_02_lf
  bone2Name=pelvis
  swing1LimitDegrees=45
  swing1Motion=Free
  swing2LimitDegrees=45
  swing2Motion=Free
  twistLimitDegrees=45
  twistMotion=Free
 }
]
