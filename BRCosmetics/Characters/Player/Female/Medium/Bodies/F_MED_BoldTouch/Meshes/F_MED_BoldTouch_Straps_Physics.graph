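# Physics asset for the left-side straps on the F_MED_BoldTouch female player body cosmetic, built entirely from capsule bodies.
# A three-bone dyn_straps chain hangs off base_straps_lf, with narrow swing limits and locked twist at each joint.
# The base links to upperarm_l through an angularly free constraint.
# Larger capsules on the upper arm and front and back jacket-side bones add body volumes; the jacket ones have no constraints.
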PhysicsAsset /BRCosmetics/Characters/Player/Female/Medium/Bodies/F_MED_BoldTouch/Meshes/F_MED_BoldTouch_Straps_Physics
bodies {
 base_straps_lf [{shapeType=Capsule, length=0.505, radius=0.505, shapeName=base_straps_lf_capsule, rotation=<rot 0 -90 0>}]
 dyn_jacket_side_back_01_lf [{
   shapeType=Capsule
   length=10.0811
   radius=8.2678
   shapeName=dyn_jacket_side_back_01_lf_capsule
   center=<9.0863 -2.7918 -4.891>
   rotation=<rot 75.0002 -179.9999 -180>
  }]
 dyn_jacket_side_front_01_lf [{
   shapeType=Capsule
   length=2.1717
   radius=12.9103
   shapeName=dyn_jacket_side_front_01_lf_capsule
   center=<11.9402 0 -10.2865>
   rotation=<rot 90 -89.9999 0>
  }]
 dyn_straps_01_lf [{shapeType=Capsule, length=4, radius=1, shapeName=dyn_straps_01_lf_capsule, center=<2 0 0>, rotation=<rot 90 0 0>}]
 dyn_straps_02_lf [{shapeType=Capsule, length=3.353, radius=1.647, shapeName=dyn_straps_02_lf_capsule, center=<2 0 0>, rotation=<rot 90 0 0>}]
 dyn_straps_03_lf [{shapeType=Capsule, length=2.8659, radius=2.1341, shapeName=dyn_straps_03_lf_capsule, center=<2 0 0>, rotation=<rot 90 0 0>}]
 upperarm_l [{
   shapeType=Capsule
   length=16.5387
   radius=6.4452
   shapeName=upperarm_l_capsule
   center=<20.8179 0.381 -0.4009>
   rotation=<rot -0.4422 -89.9999 -92.1918>
  }]
}
constraints [
 {
  bone1Name=base_straps_lf
  bone2Name=upperarm_l
  swing1LimitDegrees=45
  swing1Motion=Free
  swing2LimitDegrees=45
  swing2Motion=Free
  twistLimitDegrees=45
  twistMotion=Free
 }
 {bone1Name=dyn_straps_01_lf, bone2Name=base_straps_lf, swing1LimitDegrees=10, swing2LimitDegrees=5, twistLimitDegrees=45, twistMotion=Locked}
 {
  bone1Name=dyn_straps_02_lf
  bone2Name=dyn_straps_01_lf
  swing1LimitDegrees=15
  swing2LimitDegrees=10
  twistLimitDegrees=45
  twistMotion=Locked
 }
 {
  bone1Name=dyn_straps_03_lf
  bone2Name=dyn_straps_02_lf
  swing1LimitDegrees=15
  swing2LimitDegrees=10
  twistLimitDegrees=45
  twistMotion=Locked
 }
]
