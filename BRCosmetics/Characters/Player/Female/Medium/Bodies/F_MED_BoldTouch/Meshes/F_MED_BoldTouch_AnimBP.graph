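# Post-process style anim blueprint for the F_MED_BoldTouch cosmetic body, adding corrective and secondary motion to jacket, straps, belts and tail.
# The copied parent pose runs through a Control Rig and additive jacket/strap bone offsets, then cached poses feed AnimDynamics and RigidBody simulations.
# IsRBANTail selects between two physics assets before the straps simulation.
# The event graph hides the tail under capes or long backblings and derives movement, gravity and zipline values.
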AnimBlueprint /BRCosmetics/Characters/Player/Female/Medium/Bodies/F_MED_BoldTouch/Meshes/F_MED_BoldTouch_AnimBP
target_skeleton=&/F_MED_BoldTouch_Skeleton
graphs [
 AnimGraph {nodes [
   n1 = RigidBody!(ComponentPose=Local To Component(LocalPose=Use cached pose 'AnimDynCache'()), Alpha, OverrideWorldGravity:Vector="0,0,-980") -> (Pose) {
    bClampLinearTranslationLimitToRefPose=true
    bOverrideWorldGravity=true
    overridePhysicsAsset=&/F_MED_BoldTouch_Physics
    simulationSpace=BaseBoneSpace
    alphaBoolBlend {blendOption=Linear}
    baseBoneRef {boneName=pelvis}
    overrideWorldGravity {z=-980}
    simSpaceSettings {worldAlpha=0.4}
   }
   n2 = Component To Local(ComponentPose=@n1)
   n3 = RigidBody(ComponentPose=Local To Component(LocalPose=Use cached pose 'AnimDynCache'()), OverrideWorldGravity="0,0,-980") {
    bClampLinearTranslationLimitToRefPose=true
    bOverrideWorldGravity=true
    overridePhysicsAsset=&/F_MED_BoldTouch_Tail_Backpack_Physics
    simulationSpace=BaseBoneSpace
    alphaBoolBlend {blendOption=Linear}
    baseBoneRef {boneName=pelvis}
    overrideWorldGravity {z=-980}
    simSpaceSettings {worldAlpha=0.3}
   }
   n4 = Component To Local(ComponentPose=@n3)
   n5 = Get IsRBANTail() -> (IsRBANTail:bool)
   n6 = RigidBody!(
    ComponentPose=Local To Component(LocalPose=Blend Poses by bool(BlendPose_0=@n4, BlendPose_1=@n2, bActiveValue=@n5) {blendType=Linear})
    Alpha
    OverrideWorldGravity="0,0,-980"
    ExternalForce:Vector
   ) -> (Pose) {
    bOverrideWorldGravity=true
    overridePhysicsAsset=&/F_MED_BoldTouch_Straps_Physics
    simulationSpace=BaseBoneSpace
    alphaBoolBlend {blendOption=Linear}
    baseBoneRef {boneName=upperarm_twist_01_l}
    overrideWorldGravity {z=-980}
    simSpaceSettings {worldAlpha=0.5}
   }
   n7 = "Transform (Modify) Bone - Bone: dyn_strap_front_05_lf"!(ComponentPose=@n6, bAlphaBoolEnabled:bool=True, Translation="0,0,0", Rotation="-9.982,-29.970,0.032") -> (Pose) {
    alphaInputType=Bool
    rotationMode=BMM_Additive
    rotationSpace=BCS_BoneSpace
    translationMode=BMM_Additive
    translationSpace=BCS_BoneSpace
    alphaBoolBlend {blendInTime=0.2, blendOption=Linear, blendOutTime=0.2}
    boneToModify {boneName=dyn_strap_front_05_lf}
    rotation=<rot -9.982 -29.97 0.032>
   }
   n8 = "Transform (Modify) Bone - Bone: dyn_strap_back_04_lf"!(ComponentPose=@n7, bAlphaBoolEnabled:bool=True, Translation="0,0,0", Rotation="-5.034,-5.003,0.011") -> (Pose) {
    alphaInputType=Bool
    rotationMode=BMM_Additive
    rotationSpace=BCS_BoneSpace
    translationMode=BMM_Additive
    translationSpace=BCS_BoneSpace
    alphaBoolBlend {blendInTime=0.2, blendOption=Linear, blendOutTime=0.2}
    boneToModify {boneName=dyn_strap_back_04_lf}
    rotation=<rot -5.034 -5.003 0.011>
   }
   Output Pose(Result=Component To Local(ComponentPose=@n8)) {layerGroup=None}
   n9 = Control Rig!(
    Source=Copy Pose From Mesh() {bUseAttachedParent=true}
    Alpha
    Is PickAxe Gauntlet:bool=Get is Use Pickaxe Gauntlet() -> (is Use Pickaxe Gauntlet:bool)
   ) -> (Pose) {
    alphaBoolBlend {blendOption=Linear}
    controlRigAssetReference {blueprintRigClass=&/F_MED_BoldTouch_CtrlRig.F_MED_BoldTouch_CtrlRig_C}
   }
   n10 = "Transform (Modify) Bone - Bone: offset_strap_back_lf"!(ComponentPose=Local To Component(LocalPose=@n9), bAlphaBoolEnabled:bool=True, Translation="0,0,0", Rotation="-9.329,30.115,-2.597") -> (Pose) {
    alphaInputType=Bool
    rotationMode=BMM_Additive
    rotationSpace=BCS_BoneSpace
    translationMode=BMM_Additive
    translationSpace=BCS_BoneSpace
    alphaBoolBlend {blendInTime=0.5, blendOption=Linear, blendOutTime=0.5}
    boneToModify {boneName=offset_strap_back_lf}
    rotation=<rot -9.329 30.115 -2.597>
   }
   n11 = "Transform (Modify) Bone - Bone: base_jacket_front_lf"!(ComponentPose=@n10, bAlphaBoolEnabled:bool=True, Translation="0,0,0", Rotation="-5,-15,-14.13") -> (Pose) {
    alphaInputType=Bool
    rotationMode=BMM_Additive
    rotationSpace=BCS_BoneSpace
    translationMode=BMM_Additive
    translationSpace=BCS_BoneSpace
    alphaBoolBlend {blendInTime=0.5, blendOption=Linear, blendOutTime=0.5}
    boneToModify {boneName=base_jacket_front_lf}
    rotation=<rot -5 -15 -14.13>
   }
   n12 = "Constraint - jacket_spine_cr (R)"!(ComponentPose=@n11, bAlphaBoolEnabled:bool=True, ConstraintWeights_0=1) -> (Pose) {
    alphaInputType=Bool
    alphaBoolBlend {blendInTime=1, blendOption=Linear, blendOutTime=1}
    boneToModify {boneName=jacket_spine_cr}
    constraintSetup [{transformType=Rotation, targetBone {boneName=pelvis}}]
    constraintWeights [1]
   }
   n13 = "Transform (Modify) Bone - Bone: base_jacket_front_lf"(ComponentPose=@n12, bAlphaBoolEnabled=True, Translation="0,0,0", Rotation="0.000,-25.000,-15.000") {
    alphaInputType=Bool
    rotationMode=BMM_Additive
    rotationSpace=BCS_BoneSpace
    translationMode=BMM_Additive
    translationSpace=BCS_BoneSpace
    alphaBoolBlend {blendOption=Linear}
    boneToModify {boneName=base_jacket_front_lf}
    rotation=<rot 0 -24.9999 -15>
   }
   n14 = "Transform (Modify) Bone - Bone: base_jacket_side_front_lf"!(ComponentPose=@n13, bAlphaBoolEnabled:bool=True, Translation="0,0,0", Rotation="14.478,-15.504,-18.967") -> (Pose) {
    alphaInputType=Bool
    rotationMode=BMM_Additive
    rotationSpace=BCS_BoneSpace
    translationMode=BMM_Additive
    translationSpace=BCS_BoneSpace
    alphaBoolBlend {blendOption=Linear}
    boneToModify {boneName=base_jacket_side_front_lf}
    rotation=<rot 14.478 -15.504 -18.967>
   }
   n15 = "Transform (Modify) Bone - Bone: base_jacket_side_back_lf"!(ComponentPose=@n14, bAlphaBoolEnabled:bool=True, Translation="-2,0,0", Rotation="5,0,0") -> (Pose) {
    alphaInputType=Bool
    rotationMode=BMM_Additive
    rotationSpace=BCS_BoneSpace
    translationMode=BMM_Additive
    translationSpace=BCS_BoneSpace
    alphaBoolBlend {blendInTime=0.5, blendOption=Linear, blendOutTime=0.2}
    boneToModify {boneName=base_jacket_side_back_lf}
    rotation {pitch=5}
    translation {x=-2}
   }
   n16 = "Transform (Modify) Bone - Bone: base_jacket_back_cr"!(ComponentPose=@n15, bAlphaBoolEnabled:bool=True, Translation="-2.5,0,0", Rotation="10,0,0") -> (Pose) {
    alphaInputType=Bool
    rotationMode=BMM_Additive
    rotationSpace=BCS_BoneSpace
    translationMode=BMM_Additive
    translationSpace=BCS_BoneSpace
    alphaBoolBlend {blendInTime=0.5, blendOption=Linear, blendOutTime=0.2}
    boneToModify {boneName=base_jacket_back_cr}
    rotation {pitch=10}
    translation {x=-2.5}
   }
   n17 = "Transform (Modify) Bone - Bone: base_jacket_side_back_rt"!(ComponentPose=@n16, bAlphaBoolEnabled:bool=True, Translation="-2,0,0", Rotation="5,0,0") -> (Pose) {
    alphaInputType=Bool
    rotationMode=BMM_Additive
    rotationSpace=BCS_BoneSpace
    translationMode=BMM_Additive
    translationSpace=BCS_BoneSpace
    alphaBoolBlend {blendInTime=0.5, blendOption=Linear, blendOutTime=0.2}
    boneToModify {boneName=base_jacket_side_back_rt}
    rotation {pitch=5}
    translation {x=-2}
   }
   n18 = "Transform (Modify) Bone - Bone: attach_backpack"!(ComponentPose=@n17, bAlphaBoolEnabled:bool=True, Translation="0,0,4", Rotation="0.021,-0.102,20") -> (Pose) {
    alphaInputType=Bool
    rotationMode=BMM_Additive
    rotationSpace=BCS_BoneSpace
    translationMode=BMM_Additive
    translationSpace=BCS_BoneSpace
    alphaBoolBlend {blendInTime=0.1, blendOption=Linear, blendOutTime=1}
    boneToModify {boneName=attach_backpack}
    rotation=<rot 0.021 -0.102 20>
    translation {z=4}
   }
   n19 = "Transform (Modify) Bone - Bone: base_jacket_side_front_lf"(ComponentPose=@n18, bAlphaBoolEnabled=True, Translation="0,0,0", Rotation="10.000,-0.020,-0.014") {
    alphaInputType=Bool
    rotationMode=BMM_Additive
    rotationSpace=BCS_BoneSpace
    translationMode=BMM_Additive
    translationSpace=BCS_BoneSpace
    alphaBoolBlend {blendOption=Linear}
    boneToModify {boneName=base_jacket_side_front_lf}
    rotation=<rot 10 -0.02 -0.014>
   }
   n20 = "Transform (Modify) Bone - Bone: base_jacket_front_lf"(ComponentPose=@n19, bAlphaBoolEnabled=True, Translation="0,0,0", Rotation="-26.989,-28.500,13.693") {
    alphaInputType=Bool
    rotationMode=BMM_Additive
    rotationSpace=BCS_BoneSpace
    translationMode=BMM_Additive
    translationSpace=BCS_BoneSpace
    alphaBoolBlend {blendOption=Linear}
    boneToModify {boneName=base_jacket_front_lf}
    rotation=<rot -26.989 -28.5 13.693>
   }
   n21 = "Transform (Modify) Bone - Bone: offset_strap_front_lf"!(ComponentPose=@n20, bAlphaBoolEnabled:bool=True, Translation="0,0,0", Rotation="-0.139,-20.013,-0.003") -> (Pose) {
    alphaInputType=Bool
    rotationMode=BMM_Additive
    rotationSpace=BCS_BoneSpace
    translationMode=BMM_Additive
    translationSpace=BCS_BoneSpace
    alphaBoolBlend {blendOption=Linear}
    boneToModify {boneName=offset_strap_front_lf}
    rotation=<rot -0.139 -20.013 -0.003>
   }
   n22 = "Transform (Modify) Bone - Bone: dyn_strap_front_03_lf"!(ComponentPose=@n21, bAlphaBoolEnabled:bool=True, Translation="0,0,0", Rotation="17.172,-60.481,13.256") -> (Pose) {
    alphaInputType=Bool
    rotationMode=BMM_Additive
    rotationSpace=BCS_BoneSpace
    translationMode=BMM_Additive
    translationSpace=BCS_BoneSpace
    alphaBoolBlend {blendOption=Linear}
    boneToModify {boneName=dyn_strap_front_03_lf}
    rotation=<rot 17.172 -60.481 13.256>
   }
   n23 = "Transform (Modify) Bone - Bone: dyn_strap_front_02_lf"!(ComponentPose=@n22, bAlphaBoolEnabled:bool=True, Translation="0,0,0", Rotation="0,70,0") -> (Pose) {
    alphaInputType=Bool
    rotationMode=BMM_Additive
    rotationSpace=BCS_BoneSpace
    translationMode=BMM_Additive
    translationSpace=BCS_BoneSpace
    alphaBoolBlend {blendOption=Linear}
    boneToModify {boneName=dyn_strap_front_02_lf}
    rotation {yaw=70}
   }
   n24 = "Transform (Modify) Bone - Bone: offset_strap_back_lf"(ComponentPose=@n23, bAlphaBoolEnabled=True, Translation="0,0,0", Rotation="0,-20,0") {
    alphaInputType=Bool
    rotationMode=BMM_Additive
    rotationSpace=BCS_BoneSpace
    translationMode=BMM_Additive
    translationSpace=BCS_BoneSpace
    alphaBoolBlend {blendOption=Linear}
    boneToModify {boneName=offset_strap_back_lf}
    rotation {yaw=-20}
   }
   n25 = "Transform (Modify) Bone - Bone: base_jacket_back_cr"(ComponentPose=@n24, bAlphaBoolEnabled=True, Translation="-3,0,0", Rotation="0,0,0") {
    alphaInputType=Bool
    rotationMode=BMM_Additive
    rotationSpace=BCS_BoneSpace
    translationMode=BMM_Additive
    translationSpace=BCS_BoneSpace
    alphaBoolBlend {blendInTime=0.1, blendOption=Linear, blendOutTime=0.3}
    boneToModify {boneName=base_jacket_back_cr}
    translation {x=-3}
   }
   n26 = "Transform (Modify) Bone - Bone: base_jacket_side_back_lf"(ComponentPose=@n25, bAlphaBoolEnabled=True, Translation="-3,0,0", Rotation="0,0,0") {
    alphaInputType=Bool
    rotationMode=BMM_Additive
    rotationSpace=BCS_BoneSpace
    translationMode=BMM_Additive
    translationSpace=BCS_BoneSpace
    alphaBoolBlend {blendInTime=0.1, blendOption=Linear, blendOutTime=0.3}
    boneToModify {boneName=base_jacket_side_back_lf}
    translation {x=-3}
   }
   n27 = "Transform (Modify) Bone - Bone: base_jacket_side_back_rt"(ComponentPose=@n26, bAlphaBoolEnabled=True, Translation="-3,0,0", Rotation="0,0,0") {
    alphaInputType=Bool
    rotationMode=BMM_Additive
    rotationSpace=BCS_BoneSpace
    translationMode=BMM_Additive
    translationSpace=BCS_BoneSpace
    alphaBoolBlend {blendInTime=0.1, blendOption=Linear, blendOutTime=0.3}
    boneToModify {boneName=base_jacket_side_back_rt}
    translation {x=-3}
   }
   n28 = "Transform (Modify) Bone - Bone: base_jacket_back_cr"!(ComponentPose=@n27, bAlphaBoolEnabled=True, Rotation="10,0,0") -> (Pose) {
    alphaInputType=Bool
    rotationMode=BMM_Additive
    rotationSpace=BCS_BoneSpace
    translationMode=BMM_Additive
    translationSpace=BCS_BoneSpace
    alphaBoolBlend {blendInTime=0.15, blendOption=Linear, blendOutTime=0.2}
    alphaScaleBiasClamp {bInterpResult=true, interpSpeedDecreasing=0, interpSpeedIncreasing=0}
    boneToModify {boneName=base_jacket_back_cr}
    rotation {pitch=10}
    translation {x=-4}
   }
   n29 = "Transform (Modify) Bone - Bone: base_jacket_side_back_lf"!(ComponentPose=@n28, bAlphaBoolEnabled=True, Rotation="10,0,0") -> (Pose) {
    alphaInputType=Bool
    rotationMode=BMM_Additive
    rotationSpace=BCS_BoneSpace
    translationMode=BMM_Additive
    translationSpace=BCS_BoneSpace
    alphaBoolBlend {blendInTime=0.15, blendOption=Linear, blendOutTime=0.2}
    alphaScaleBiasClamp {bInterpResult=true, interpSpeedDecreasing=0, interpSpeedIncreasing=0}
    boneToModify {boneName=base_jacket_side_back_lf}
    rotation {pitch=10}
    translation {x=-2}
   }
   n30 = "Transform (Modify) Bone - Bone: base_jacket_side_back_rt"!(ComponentPose=@n29, bAlphaBoolEnabled=True, Rotation="10,0,0") -> (Pose) {
    alphaInputType=Bool
    rotationMode=BMM_Additive
    rotationSpace=BCS_BoneSpace
    translationMode=BMM_Additive
    translationSpace=BCS_BoneSpace
    alphaBoolBlend {blendInTime=0.15, blendOption=Linear, blendOutTime=0.2}
    alphaScaleBiasClamp {bInterpResult=true, interpSpeedDecreasing=0, interpSpeedIncreasing=0}
    boneToModify {boneName=base_jacket_side_back_rt}
    rotation {pitch=10}
    translation {x=-2}
   }
   n31 = "Transform (Modify) Bone - Bone: base_jacket_front_lf"(ComponentPose=@n30, bAlphaBoolEnabled=True, Translation="0,0,0", Rotation="7.000,0.112,0.046") {
    alphaInputType=Bool
    rotationMode=BMM_Additive
    rotationSpace=BCS_BoneSpace
    translationMode=BMM_Additive
    translationSpace=BCS_BoneSpace
    alphaBoolBlend {blendOption=Linear}
    boneToModify {boneName=base_jacket_front_lf}
    rotation=<rot 7 0.112 0.046>
   }
   n32 = "Transform (Modify) Bone - Bone: base_jacket_side_front_rt"!(ComponentPose=@n31, bAlphaBoolEnabled:bool=True, Translation="-2,0,0", Rotation="0,0,0") -> (Pose) {
    alphaInputType=Bool
    rotationMode=BMM_Additive
    rotationSpace=BCS_BoneSpace
    translationMode=BMM_Additive
    translationSpace=BCS_BoneSpace
    alphaBoolBlend {blendInTime=0.2, blendOption=Linear, blendOutTime=0.2}
    boneToModify {boneName=base_jacket_side_front_rt}
    translation {x=-2}
   }
   n33 = "Transform (Modify) Bone - Bone: base_jacket_side_back_lf"!(ComponentPose=@n32, bAlphaBoolEnabled=True, Translation:Vector="4,0,0", Rotation="5,0,0") -> (Pose) {
    alphaInputType=Bool
    rotationMode=BMM_Additive
    rotationSpace=BCS_BoneSpace
    translationMode=BMM_Additive
    translationSpace=BCS_BoneSpace
    alphaBoolBlend {blendInTime=0.2, blendOption=Linear, blendOutTime=0.2}
    boneToModify {boneName=base_jacket_side_back_lf}
    rotation {pitch=5}
    translation {x=4}
   }
   n34 = "Transform (Modify) Bone - Bone: base_jacket_back_cr"!(ComponentPose=@n33, bAlphaBoolEnabled=True, Translation:Vector="6,0,0", Rotation="0,0,0") -> (Pose) {
    alphaInputType=Bool
    rotationMode=BMM_Additive
    rotationSpace=BCS_BoneSpace
    translationMode=BMM_Additive
    translationSpace=BCS_BoneSpace
    alphaBoolBlend {blendInTime=0.2, blendOption=Linear, blendOutTime=0.2}
    boneToModify {boneName=base_jacket_back_cr}
    translation {x=6}
   }
   n35 = "Transform (Modify) Bone - Bone: base_jacket_side_back_rt"!(ComponentPose=@n34, bAlphaBoolEnabled=True, Translation:Vector="3,0,0", Rotation="0,0,0") -> (Pose) {
    alphaInputType=Bool
    rotationMode=BMM_Additive
    rotationSpace=BCS_BoneSpace
    translationMode=BMM_Additive
    translationSpace=BCS_BoneSpace
    alphaBoolBlend {blendInTime=0.2, blendOption=Linear, blendOutTime=0.2}
    boneToModify {boneName=base_jacket_side_back_rt}
    translation {x=3}
   }
   n36 = "Transform (Modify) Bone - Bone: base_jacket_side_front_rt"(ComponentPose=@n35, bAlphaBoolEnabled=True, Translation="-1,0,0", Rotation="0,0,0") {
    alphaInputType=Bool
    rotationMode=BMM_Additive
    rotationSpace=BCS_BoneSpace
    translationMode=BMM_Additive
    translationSpace=BCS_BoneSpace
    alphaBoolBlend {blendInTime=0.2, blendOption=Linear, blendOutTime=0.2}
    boneToModify {boneName=base_jacket_side_front_rt}
    translation {x=-1}
   }
   Save cached pose 'FixedCache'(Pose=Component To Local(ComponentPose=@n36)) {cacheName=FixedCache}
   n37 = Anim Dynamics - Bone: lowerarm_belt_rt(ComponentPose=Local To Component(LocalPose=Use cached pose 'SlideCache'()), Alpha=1, ExternalForce="0,0,0") {
    angularDampingOverride=1
    angularSpringConstant=2000
    bAngularSpring=true
    bOverrideAngularDamping=true
    bUseGravityOverride=true
    gravityScale=0.35
    simulationSpace=RootRelative
    alphaBoolBlend {blendOption=Linear}
    boundBone {boneName=lowerarm_belt_rt}
    componentAppliedLinearAccClamp=<400 400 400>
    componentLinearAccScale=<1 1 1>
    componentLinearVelScale=<1 1 1>
    gravityOverride {z=-980}
    physicsBodyDefinitions [{
      boundBone {boneName=lowerarm_belt_rt}
      boxExtents=<10 10 10>
      constraintSetup {angularLimitsMax=<0 5 0>, angularLimitsMin=<0 -5 0>, angularTarget=<1 0 0>}
      localJointOffset=<2 2 0>
     }]
   }
   n38 = Anim Dynamics - Bone: lowerarm_belt_lf(ComponentPose=@n37, Alpha=1, ExternalForce="0,0,0") {
    angularDampingOverride=1
    angularSpringConstant=2000
    bAngularSpring=true
    bOverrideAngularDamping=true
    bUseGravityOverride=true
    gravityScale=0.35
    simulationSpace=RootRelative
    alphaBoolBlend {blendOption=Linear}
    boundBone {boneName=lowerarm_belt_lf}
    componentAppliedLinearAccClamp=<400 400 400>
    componentLinearAccScale=<1 1 1>
    componentLinearVelScale=<1 1 1>
    gravityOverride {z=-980}
    physicsBodyDefinitions [{
      boundBone {boneName=lowerarm_belt_lf}
      boxExtents=<10 10 10>
      constraintSetup {angularLimitsMax=<0 5 0>, angularLimitsMin=<0 -5 0>, angularTarget=<1 0 0>}
      localJointOffset=<2 2 0>
     }]
   }
   n39 = Anim Dynamics - Bone: dyn_straps_01_rt!(ComponentPose=@n38, Alpha=1, ComponentLinearAccScale:Vector="1,1,1", ComponentLinearVelScale:Vector="1,1,1", ExternalForce="0,0,0") -> (Pose) {
    angularDampingOverride=1
    angularSpringConstant=10000
    bAngularSpring=true
    bOverrideAngularDamping=true
    bUseGravityOverride=true
    gravityScale=0.35
    simulationSpace=RootRelative
    alphaBoolBlend {blendOption=Linear}
    boundBone {boneName=dyn_straps_01_rt}
    componentAppliedLinearAccClamp=<3000 3000 3000>
    componentLinearAccScale=<1 1 1>
    componentLinearVelScale=<1 1 1>
    gravityOverride {z=-980}
    physicsBodyDefinitions [{
      boundBone {boneName=dyn_straps_01_rt}
      boxExtents=<20 30 30>
      constraintSetup {angularLimitsMax=<0 5 0>, angularLimitsMin=<0 -5 -20>, angularTarget=<1 0 0>}
      localJointOffset=<4 0 0>
     }]
   }
   n40 = Anim Dynamics - Bone: dyn_straps_02_rt!(ComponentPose=@n39, Alpha=1, ComponentLinearAccScale:Vector="1,1,1", ComponentLinearVelScale:Vector="1,1,1", ExternalForce="0,0,0") -> (Pose) {
    angularDampingOverride=1
    angularSpringConstant=10000
    bAngularSpring=true
    bOverrideAngularDamping=true
    bUseGravityOverride=true
    gravityScale=0.35
    simulationSpace=RootRelative
    alphaBoolBlend {blendOption=Linear}
    boundBone {boneName=dyn_straps_02_rt}
    componentAppliedLinearAccClamp=<3000 3000 3000>
    componentLinearAccScale=<1 1 1>
    componentLinearVelScale=<1 1 1>
    gravityOverride {z=-980}
    physicsBodyDefinitions [{
      boundBone {boneName=dyn_straps_02_rt}
      boxExtents=<30 30 30>
      constraintSetup {angularLimitsMax=<0 20 4>, angularLimitsMin=<0 -20 -15>, angularTarget=<1 0 0>}
      localJointOffset=<4 0 0>
     }]
   }
   n41 = Anim Dynamics - Bone: dyn_straps_03_rt!(ComponentPose=@n40, Alpha=1, ComponentLinearAccScale:Vector="1,1,1", ComponentLinearVelScale:Vector="1,1,1", ExternalForce="0,0,0") -> (Pose) {
    angularDampingOverride=1
    angularSpringConstant=10000
    bAngularSpring=true
    bOverrideAngularDamping=true
    bUseGravityOverride=true
    gravityScale=0.35
    simulationSpace=RootRelative
    alphaBoolBlend {blendOption=Linear}
    boundBone {boneName=dyn_straps_03_rt}
    componentAppliedLinearAccClamp=<3000 3000 3000>
    componentLinearAccScale=<1 1 1>
    componentLinearVelScale=<1 1 1>
    gravityOverride {z=-980}
    physicsBodyDefinitions [{
      boundBone {boneName=dyn_straps_03_rt}
      boxExtents=<20 30 30>
      constraintSetup {angularLimitsMax=<0 20 0>, angularLimitsMin=<0 -20 -15>, angularTarget=<1 0 0>}
      localJointOffset=<4 0 0>
     }]
   }
   n42 = RigidBody(ComponentPose=@n41, OverrideWorldGravity="0,0,-980") {
    bOverrideWorldGravity=true
    overridePhysicsAsset=&/F_MED_BoldTouch_Belt_Physics
    simulationSpace=BaseBoneSpace
    alphaBoolBlend {blendOption=Linear}
    baseBoneRef {boneName=thigh_twist_01_l}
    overrideWorldGravity {z=-980}
    simSpaceSettings {worldAlpha=0.3}
   }
   Save cached pose 'AnimDynCache'(Pose=Component To Local(ComponentPose=@n42)) {cacheName=AnimDynCache}
   n43 = "Transform (Modify) Bone - Bone: base_jacket_side_back_lf"(
    ComponentPose=Local To Component(LocalPose=Use cached pose 'FixedCache'())
    bAlphaBoolEnabled=True
    Translation="0,0,0"
    Rotation="15.088,0.443,-1.953"
   ) {
    alphaInputType=Bool
    rotationMode=BMM_Additive
    rotationSpace=BCS_BoneSpace
    translationMode=BMM_Additive
    translationSpace=BCS_BoneSpace
    alphaBoolBlend {blendInTime=0.5, blendOption=Linear, blendOutTime=0.2}
    boneToModify {boneName=base_jacket_side_back_lf}
    rotation=<rot 15.088 0.443 -1.953>
   }
   n44 = "Transform (Modify) Bone - Bone: base_jacket_back_cr"(ComponentPose=@n43, bAlphaBoolEnabled=True, Translation="0,0,0", Rotation="39.665,-6.432,-4.215") {
    alphaInputType=Bool
    rotationMode=BMM_Additive
    rotationSpace=BCS_BoneSpace
    translationMode=BMM_Additive
    translationSpace=BCS_BoneSpace
    alphaBoolBlend {blendInTime=0.5, blendOption=Linear, blendOutTime=0.2}
    boneToModify {boneName=base_jacket_back_cr}
    rotation=<rot 39.665 -6.432 -4.215>
   }
   n45 = "Transform (Modify) Bone - Bone: base_jacket_side_back_rt"(ComponentPose=@n44, bAlphaBoolEnabled=True, Translation="-1.514,0.001,-0.894", Rotation="59.656,-0.066,-0.217") {
    alphaInputType=Bool
    rotationMode=BMM_Additive
    rotationSpace=BCS_BoneSpace
    translationMode=BMM_Additive
    translationSpace=BCS_BoneSpace
    alphaBoolBlend {blendInTime=0.5, blendOption=Linear, blendOutTime=0.2}
    boneToModify {boneName=base_jacket_side_back_rt}
    rotation=<rot 59.656 -0.066 -0.217>
    translation=<-1.514 0.001 -0.894>
   }
   n46 = "Transform (Modify) Bone - Bone: base_jacket_side_front_rt"(ComponentPose=@n45, bAlphaBoolEnabled=True, Translation="-2.585,0.005,-4.288", Rotation="69.566,-0.265,-0.381") {
    alphaInputType=Bool
    rotationMode=BMM_Additive
    rotationSpace=BCS_BoneSpace
    translationMode=BMM_Additive
    translationSpace=BCS_BoneSpace
    alphaBoolBlend {blendInTime=0.5, blendOption=Linear, blendOutTime=0.2}
    boneToModify {boneName=base_jacket_side_front_rt}
    rotation=<rot 69.566 -0.265 -0.381>
    translation=<-2.585 0.005 -4.288>
   }
   n47 = "Transform (Modify) Bone - Bone: base_jacket_front_rt"!(ComponentPose=@n46, bAlphaBoolEnabled:bool=True, Translation="0,0,0", Rotation="20.285,37.494,15.053") -> (Pose) {
    alphaInputType=Bool
    rotationMode=BMM_Additive
    rotationSpace=BCS_BoneSpace
    translationMode=BMM_Additive
    translationSpace=BCS_BoneSpace
    alphaBoolBlend {blendInTime=0.5, blendOption=Linear, blendOutTime=0.2}
    boneToModify {boneName=base_jacket_front_rt}
    rotation=<rot 20.285 37.494 15.053>
   }
   n48 = "Transform (Modify) Bone - Bone: base_jacket_side_back_rt"(ComponentPose=@n47, bAlphaBoolEnabled=True, Translation="-4.5,0,0", Rotation="0,0,0") {
    alphaInputType=Bool
    rotationMode=BMM_Additive
    rotationSpace=BCS_BoneSpace
    translationMode=BMM_Additive
    translationSpace=BCS_BoneSpace
    alphaBoolBlend {blendInTime=0.5, blendOption=Linear, blendOutTime=0.2}
    boneToModify {boneName=base_jacket_side_back_rt}
    translation {x=-4.5}
   }
   n49 = "Transform (Modify) Bone - Bone: base_jacket_side_front_rt"(ComponentPose=@n48, bAlphaBoolEnabled=True, Translation="-7,0,0", Rotation="0,0,0") {
    alphaInputType=Bool
    rotationMode=BMM_Additive
    rotationSpace=BCS_BoneSpace
    translationMode=BMM_Additive
    translationSpace=BCS_BoneSpace
    alphaBoolBlend {blendInTime=0.5, blendOption=Linear, blendOutTime=0.2}
    boneToModify {boneName=base_jacket_side_front_rt}
    translation {x=-7}
   }
   n50 = "Transform (Modify) Bone - Bone: base_jacket_back_cr"(ComponentPose=@n49, bAlphaBoolEnabled=True, Translation="-3,0,0", Rotation="0,0,0") {
    alphaInputType=Bool
    rotationMode=BMM_Additive
    rotationSpace=BCS_BoneSpace
    translationMode=BMM_Additive
    translationSpace=BCS_BoneSpace
    alphaBoolBlend {blendInTime=0.5, blendOption=Linear, blendOutTime=0.2}
    boneToModify {boneName=base_jacket_back_cr}
    translation {x=-3}
   }
   n51 = "Transform (Modify) Bone - Bone: base_jacket_side_back_lf"(ComponentPose=@n50, bAlphaBoolEnabled=True, Translation="-2,0,0", Rotation="0,0,0") {
    alphaInputType=Bool
    rotationMode=BMM_Additive
    rotationSpace=BCS_BoneSpace
    translationMode=BMM_Additive
    translationSpace=BCS_BoneSpace
    alphaBoolBlend {blendInTime=0.5, blendOption=Linear, blendOutTime=0.2}
    boneToModify {boneName=base_jacket_side_back_lf}
    translation {x=-2}
   }
   Save cached pose 'SlideCache'(Pose=Component To Local(ComponentPose=@n51)) {cacheName=SlideCache}
  ]}
 EventGraph {nodes [
   n1 = Event OnCharacterCustomizationLoaded()
   n2 = CheckForBackbling()
   > Set isWearingBackbling(isWearingBackbling=AND Boolean!(A=@n2, B=NOT Boolean(A=@n2.bIsBackblingACape), C:bool=NOT Boolean(A=@n2.bIsBackblingLong)) -> (ReturnValue))
   > Set Is Wearing Cape(Is Wearing Cape=OR Boolean(A=@n2.bIsBackblingACape, B=@n2.bIsBackblingLong))
   > Set TailMesh(TailMesh:Skeletal Mesh Component Object Reference=Get Array Item with EFort Custom Part Type(Array=@n1.CustomMeshParts, PartType=MiscOrTail)) -> (Output_Get:Skeletal Mesh Component Object Reference)
   > n3 = Sequence()
   n4 = CheckForBackbling()
   > Set isWearingCape(isWearingCape=OR Boolean(A=@n4.bIsBackblingACape, B=@n4.bIsBackblingLong))
   > Set characterHasTail(characterHasTail:bool=CheckForTail!(self:Custom Character Part Anim Instance Object Reference) -> (bHasTail:bool)) -> (Output_Get:bool)
   Branch(execute=@n3.then_1, Condition=Get characterHasTail() -> (characterHasTail:bool))
   > n5 = Branch(Condition=Get isWearingCape() -> (isWearingCape:bool))
   n6 = Get TailMesh() -> (TailMesh:Skeletal Mesh Component Object Reference)
   > SetVisibility!(execute, self:Scene Component Object Reference=@n6, bNewVisibility:bool=false, bPropagateToChildren:bool=false) -> (then)
   > Set IsRBANTail(IsRBANTail:bool=false) -> (Output_Get:bool)
   SetVisibility(execute=@n5.else, self=@n6, bNewVisibility=true, bPropagateToChildren=false)
   > Set IsRBANTail(IsRBANTail=true)
   Event BlueprintUpdateAnimation()
   > n7 = Sequence!(execute) -> (then_0, then_1, then_2:Exec, then_3:Exec, then_4:Exec)
   n8 = Get bIsCrouchSprinting()
   Set is Use Pickaxe Gauntlet(
    execute=@n7.then_1
    is Use Pickaxe Gauntlet:bool=AND Boolean!(
     A=Get bIsLeftArmHidden() -> (bIsLeftArmHidden:bool)
     B=Get bIsRightArmHidden() -> (bIsRightArmHidden:bool)
     C=Get bIsCrouching()
     D:bool=Get bIsPlayingMeleeAnim()
     E:bool=NOT Boolean(A=@n8)
    ) -> (ReturnValue)
   ) -> (Output_Get:bool)
   n9 = Get bRigidBodyEnabled()
   > Set IsCrouchNotRBAN(IsCrouchNotRBAN:bool=AND Boolean!(A=NOT Boolean(A=@n9), B=Get bIsCrouching()) -> (ReturnValue)) -> (Output_Get:bool)
   n10 = Get bRigidBodyEnabled()
   > Set IsCrochingWalkingNotRBAN(IsCrochingWalkingNotRBAN:bool=AND Boolean!(A=NOT Boolean(A=@n10), B=Get bIsCrouching(), C:bool=Get bIsCrouchSprinting()) -> (ReturnValue)) -> (Output_Get:bool)
   > Set IsCrouchWalkingTal(IsCrouchWalkingTal:bool=AND Boolean!(A=Get bIsCrouchSprinting(), B=Get IsRBANTail()) -> (ReturnValue)) -> (Output_Get:bool)
   n11 = Get bIsCrouchSprinting()
   > Set Is Stand Crouch RBAN(Is Stand Crouch RBAN:bool=AND Boolean!(A=Get bRigidBodyEnabled(), B=Get bIsCrouching(), C:bool=NOT Boolean(A=@n11), D:bool=Get IsRBANTail()) -> (ReturnValue)) -> (Output_Get:bool)
   n12 = PlayerMovement(execute=@n7.then_4, AnimInstance=Self-Reference(), XDotProductCoefficient=100, YDotProductCoefficient=100, ZDotProductCoefficient=100)
   > n13 = Branch(Condition=AND Boolean!(A=Get bIsCrouching(), B=Get IsRBANTail()) -> (ReturnValue))
   > Set WalkBehind(WalkBehind:double=0) -> (Output_Get:double)
   > Set WalkBoold(WalkBoold:bool=false) -> (Output_Get:bool)
   n14 = Remap Float(Input=BreakVector(InVec=MakeVector(X=@n12, Y=@n12.YMovement)).Y, OldRange=(X=0,Y=-100), NewRange=(X=0,Y=1), Clamp=true) -> (Output=~)
   Set WalkBehind(execute=@n13.else, WalkBehind=@n14)
   n15 = Get bIsTargeting()
   > n16 = Set WalkBoold(WalkBoold=AND Boolean(A=NOT Boolean(A=@n15), B="float > float"(A=@n14, B=0.1)))
   > Set jumpingCustom(jumpingCustom:bool=AND Boolean(A=NOT Boolean(A=@n16), B=Get bIsJumping())) -> (Output_Get:bool)
   n17 = Get bIsRightArmHidden()
   n18 = Get bIsLeftArmHidden()
   Set isCrouchingPickaxe(
    execute=@n7.then_2
    isCrouchingPickaxe:bool=AND Boolean!(A=Get bIsCrouching(), B=Get bIsPlayingMeleeAnim(), C:bool=Get bIsMoving2D(), D:bool=NOT Boolean(A=@n18), E:bool=NOT Boolean(A=@n17)) -> (ReturnValue)
   ) -> (Output_Get:bool)
   n19 = Get bRigidBodyEnabled()
   n20 = Get bIsCrouching()
   Set CrouchRBAN(execute=@n7.then_3, CrouchRBAN:bool=AND Boolean!(A=NOT Boolean(A=@n19), B=@n20) -> (ReturnValue)) -> (Output_Get:bool)
   > Set Is Crouch RBAN On(Is Crouch RBAN On:bool=AND Boolean(A=@n19, B=@n20)) -> (Output_Get:bool)
   n21 = "Gravity Override State Function (Parameter Asset) (Nativized)"(
    execute=@n7.then_0
    CharacterPartAnimInstance=Self-Reference()
    GameTimeInSeconds=GetGameTimeInSeconds()
    Parameters=&/F_MED_BoldTouch_Parameters
   )
   > Set GravityOverride(GravityOverride=@n21)
   n22 = Event BlueprintProcessAnimInputEvent()
   n23 = IsNamed(self=@n22.AnimInputEvent, InEventName=aie.Player.Zipline.Entered)
   n24 = OR Boolean(A=Get bIsMoving2D(), B=Get bIsTacticalSprinting())
   n25 = Get bIsJumping()
   > n26 = Branch(Condition=OR Boolean(A=@n23, B=AND Boolean!(A=@n23, B=@n25, C:bool=@n24) -> (ReturnValue)))
   > Set animDyn on zipline(animDyn on zipline:Vector="0, 0, 0") -> (Output_Get:Vector)
   Branch(execute=@n26.else, Condition=IsNamed(self=@n22.AnimInputEvent, InEventName=aie.Player.Zipline.Exited))
   > Set animDyn on zipline(animDyn on zipline="1,1,1")
   Get Gravity Override Params()
  ]}
]
variables [
 animDyn on zipline {type=Vector, default=<1 1 1>}
 characterHasTail {type=boolean}
 CrouchRBAN {type=boolean}
 Gravity Override Params {
  type=GravityOverrideParamsStruct
  default {
   skydive_DivingDown {gravityOverride=<-1500 0 -1000>}
   skydive_FloatingDown {gravityOverride=<0 -700 -200>}
   skydive_GoingUp {gravityOverride=<0 -300 -500>}
  }
 }
 GravityOverride {type=Vector}
 Is Crouch RBAN On {type=boolean}
 Is Stand Crouch RBAN {type=boolean}
 is Use Pickaxe Gauntlet {type=boolean}
 Is Wearing Cape {type=boolean}
 IsCrochingWalkingNotRBAN {type=boolean}
 isCrouchingPickaxe {type=boolean}
 IsCrouchNotRBAN {type=boolean}
 IsCrouchWalkingTal {type=boolean}
 IsRBANTail {type=boolean}
 isWearingBackbling {type=boolean}
 isWearingCape {type=boolean}
 jumpingCustom {type=boolean}
 TailMesh {type=SkeletalMeshComponent}
 walk {type=boolean}
 WalkBehind {type=number}
 WalkBoold {type=boolean}
]
