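# Control Rig for the F_MED_BoldTouch face accessory, posing hair base and dynamic bones and a neck bell bone from body motion.
# A sequence runs head-rotation-weighted hair offsets (partly using head_up), a spherical pose reader on head relative to clavicle_r that drives more hair offsets, and a weighted parent constraint for the bell.
# A clamped head-times-clavicle corrective exists but its execute pin is unconnected.
# The pose reader's debug draw is on.
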ControlRigBlueprint /BRCosmetics/Characters/Player/Female/Medium/Bodies/F_MED_BoldTouch/Meshes/Parts/F_MED_BoldTouch_FaceAcc_CtrlRig
graphs [
 Multiply_2_1_ContainedGraph {nodes [
   n1 = RigVMFunctionEntryNode()
   RigVMFunctionReturnNode(Result=&RigVMFunction_MathDoubleMul(B=@n1.C, A=&RigVMFunction_MathDoubleMul(A=@n1.A, B=@n1.B).Result).Result)
  ]}
 New Function_ContainedGraph {nodes [
   RigVMFunctionEntryNode()
   RigUnit_GetRelativeTransformForItem = &RigUnit_GetRelativeTransformForItem(Child=(Type=Bone,Name="head"), Parent=(Type=Bone,Name="spine_05"), bChildInitial=False, bParentInitial=False)
   RigVMFunction_MathQuaternionToEuler = &RigVMFunction_MathQuaternionToEuler(RotationOrder=ZYX, Value=@RigUnit_GetRelativeTransformForItem.RelativeTransform.Rotation)
   > RigVMFunctionReturnNode(Result=@RigVMFunction_MathQuaternionToEuler.Result)
  ]}
 RigVMFunctionLibrary {nodes [F_Head Rotation = RigVMCollapseNode()]}
 RigVMFunction_Sequence_1_ContainedGraph {nodes [
   n1 = RigVMFunctionEntryNode()
   RigVMFunction_Sequence = &RigVMFunction_Sequence(ExecuteContext=@n1.ExecuteContext)
   n2 = &RigVMFunction_Sequence(ExecuteContext=@RigVMFunction_Sequence.B)
   n3 = &RigVMFunction_Sequence(ExecuteContext=@n2.B)
   RigVMFunctionReturnNode(A=@RigVMFunction_Sequence.A, B=@n2.A, C=@n3.A, D=@n3.B)
  ]}
 RigVMModel {nodes [
   RigUnit_BeginExecution = &RigUnit_BeginExecution()
   RigVMFunction_Sequence = RigVMAggregateNode(ExecuteContext=@RigUnit_BeginExecution.ExecutePin)
   n1 = &RigUnit_SphericalPoseReader(
    ActiveRegionScaleFactors=(PositiveWidth=1,NegativeWidth=1,PositiveHeight=1,NegativeHeight=1)
    ActiveRegionSize=0.01
    Debug=(bDrawDebug=true,bDraw2D=False,bDrawLocalAxes=False,DebugScale=25,DebugSegments=20,DebugThickness=0.25)
    DriverAxis=(X=0,Y=1,Z=0)
    DriverItem=(Type=Bone,Name="head")
    FalloffRegionScaleFactors=(PositiveWidth=1,NegativeWidth=1,PositiveHeight=1,NegativeHeight=1)
    FalloffSize=0.1
    FlipHeightScaling=False
    FlipWidthScaling=False
    OptionalParentItem=(Type=Bone,Name="clavicle_r")
    RotationOffset=(X=0,Y=-87,Z=-5)
    ExecutePin=@RigVMFunction_Sequence.C
   )
   > &RigUnit_ModifyTransforms(
    ItemToModify="((Item=(Type=Bone,Name=\"base_hair_back_C_cr\"),Transform=(Rotation=(X=0,Y=0,Z=0,W=1),Translation=(X=0,Y=0,Z=0),Scale3D=(X=1,Y=1,Z=1))),(Item=(Type=Bone,Name=\"base_hair_back_B_lf\"),Transform=(Rotation=(X=0,Y=0,Z=0,W=1),Translation=(X=0,Y=0,Z=0),Scale3D=(X=1, ...(truncated)"
    WeightMinimum=0
    Weight=@n1.OutputParam
   )
   &RigUnit_ParentConstraint(
    AdvancedSettings=(RotationOrderForFilter=XZY)
    Child=(Type=Bone,Name="base_neck_bell_cr")
    Parents="((Item=(Type=Bone,Name=\"FACIAL_R_UnderChin\"),Weight=0.0004),(Item=(Type=Bone,Name=\"clavicle_r\"),Weight=0.003),(Item=(Type=Bone,Name=\"FACIAL_C_UnderChin\"),Weight=0.0108),(Item=(Type=Bone,Name=\"FACIAL_C_Jawline1\"),Weight=0.0337),(Item=(Type=Bone,Name=\"head\") ...(truncated)"
    Weight=1
    ExecutePin=@RigVMFunction_Sequence.D
   )
   F_Head Rotation = RigVMFunctionReferenceNode()
   Remap = &RigVMFunction_MathDoubleRemap(SourceMaximum=-80, SourceMinimum=0, TargetMaximum=1, TargetMinimum=0, Value=@"F_Head Rotation".Result.Z)
   n2 = RigVMVariableNode() {variable=head_up}
   &RigUnit_ModifyTransforms(
    ItemToModify="((Item=(Type=Bone,Name=\"dyn_hair_back_01_lf\"),Transform=(Rotation=(X=0,Y=0,Z=-0.1736,W=0.9848),Translation=(X=-5,Y=0,Z=0),Scale3D=(X=1,Y=1,Z=1))),(Item=(Type=Bone,Name=\"base_hair_back_rt\"),Transform=(Rotation=(X=0,Y=0,Z=0,W=1),Translation=(X=-5,Y=0,Z=0),Sc ...(truncated)"
    WeightMinimum=0
    Weight=@Remap.Result
    ItemToModify.0.Transform.Translation.X=@n2.Value
    ItemToModify.1.Transform.Translation.X=@n2.Value
    ItemToModify.2.Transform.Translation.X=@n2.Value
    ExecutePin=@RigVMFunction_Sequence.A
   )
   > &RigUnit_ModifyTransforms(
    ItemToModify="((Item=(Type=Bone,Name=\"dyn_hair_back_01_cr\"),Transform=(Rotation=(X=0,Y=0,Z=-0.1736,W=0.9848),Translation=(X=0,Y=0,Z=0),Scale3D=(X=1,Y=1,Z=1))),(Item=(Type=Bone,Name=\"base_hair_back_rt\"),Transform=(Rotation=(X=0,Y=0,Z=-0.1736,W=0.9848),Translation=(X=0,Y= ...(truncated)"
    WeightMinimum=0
    Weight=@Remap.Result
   )
   F_Head Rotation_4 = RigVMFunctionReferenceNode()
   Power = /Script/RigVM.RigVMFunction_MathDoublePow(
    B=0.7
    A=&RigVMFunction_MathDoubleRemap(SourceMaximum=-55, SourceMinimum=0, TargetMaximum=1, TargetMinimum=0, Value=@"F_Head Rotation_4".Result.Z).Result
   ) {is=struct RigVMUnitNode, template="Power::Execute(in A,in B,out Result)"}
   > &RigUnit_ModifyTransforms(
    ItemToModify="((Item=(Type=Bone,Name=\"base_hair_back_rt\"),Transform=(Rotation=(X=0,Y=0,Z=0,W=1),Translation=(X=0,Y=0,Z=0),Scale3D=(X=1,Y=1,Z=1))),(Item=(Type=Bone,Name=\"base_hair_back_cr\"),Transform=(Rotation=(X=0,Y=0,Z=0,W=1),Translation=(X=0,Y=0,Z=0),Scale3D=(X=1,Y=1, ...(truncated)"
    WeightMinimum=0
    Weight=@Power.Result
   )
   n3 = &RigVMFunction_MathQuaternionToEuler(
    RotationOrder=ZYX
    Value=&RigUnit_GetRelativeTransformForItem(Child=(Type=Bone,Name="head"), Parent=(Type=Null,Name="head_null"), bChildInitial=False, bParentInitial=False).RelativeTransform.Rotation
   )
   n4 = &RigUnit_GetRelativeTransformForItem(Child=(Type=Bone,Name="clavicle_r"), Parent=(Type=Null,Name="clavicle_r_null"), bChildInitial=False, bParentInitial=False)
   Multiply_2 = RigVMAggregateNode(
    A=&RigVMFunction_MathDoubleRemap(SourceMaximum=-29, SourceMinimum=-5, TargetMaximum=1, TargetMinimum=0, Value=@n3.Result.Y).Result
    B=&RigVMFunction_MathDoubleRemap(SourceMaximum=-20, SourceMinimum=5, TargetMaximum=1, TargetMinimum=0, Value=@n3.Result.Z).Result
    C=&RigVMFunction_MathDoubleRemap(
     SourceMaximum=-14
     SourceMinimum=0
     TargetMaximum=1
     TargetMinimum=0
     bClamp=False
     Value=&RigVMFunction_MathQuaternionToEuler(RotationOrder=ZYX, Value=@n4.RelativeTransform.Rotation).Result.Y
    ).Result
   )
   Clamp = &RigVMFunction_MathDoubleClamp(Maximum=0.4, Value=@Multiply_2.Result)
   &RigUnit_ModifyTransforms(
    ExecutePin=()
    ItemToModify="((Item=(Type=Bone,Name=\"dyn_hair_back_01_rt\"),Transform=(Rotation=(X=0.0151,Y=0.173,Z=-0.0858,W=0.9811),Translation=(X=0,Y=0,Z=0),Scale3D=(X=1,Y=1,Z=1))),(Item=(Type=Bone,Name=\"base_hair_back_cr\"),Transform=(Rotation=(X=0,Y=0,Z=0,W=1),Translation=(X=0,Y=0, ...(truncated)"
    WeightMinimum=0
    Weight=@Clamp.Result
   )
   RigVMRerouteNode(Value=@RigVMFunction_Sequence.B)
  ]}
]
hierarchy [
 root {loc=<0 0 0>, rot=<0 0 0 1>}
 pelvis {parent=root, loc=<0 -0.5301 98.0571>, rot=<0.0127 0.707 -0.0127 -0.707>}
 spine_01 {parent=pelvis, loc=<6.1208 -0.9328 0>, rot=<0 0 0.034 -0.9994>}
 spine_02 {parent=spine_01, loc=<6.5997 -0.2095 0>, rot=<0 0 0.0238 -0.9997>}
 spine_03 {parent=spine_02, loc=<6.4914 -0.1237 0>, rot=<0 0 0.0149 -0.9999>}
 spine_04 {parent=spine_03, loc=<6.9598 0.0794 0>, rot=<0 0 -0.0132 -0.9999>}
 spine_05 {parent=spine_04, loc=<6.9987 0 0>, rot=<0 0 -0.0015 -1>}
 neck_01 {parent=spine_05, loc=<11.3138 0.1529 0>, rot=<0 0 -0.2003 -0.9797>}
 neck_02 {parent=neck_01, loc=<4.5627 0 0>, rot=<0 0 0.0504 -0.9987>}
 head {parent=neck_02, loc=<5.3914 0 0>, rot=<0 0 0.115 -0.9934>}
 base_hair_front_B_rt {parent=head, loc=<-0.5354 2.2499 7.5074>, rot=<-0.071 -0.9066 -0.3754 -0.1793>}
 dyn_hair_front_B_01_rt {parent=base_hair_front_B_rt, loc=<0 0 0>, rot=<0 0 0 1>}
 base_hair_front_A_rt {parent=head, loc=<2.3352 4.1212 7.8618>, rot=<-0.014 -0.8679 -0.4656 -0.1723>}
 dyn_hair_front_A_01_rt {parent=base_hair_front_A_rt, loc=<0 0 0>, rot=<0 0 0 1>}
 dyn_hair_front_A_02_rt {parent=dyn_hair_front_A_01_rt, loc=<3.9462 0 0>, rot=<0 0 0 1>}
 base_hair_front_B_lf {parent=head, loc=<-3.4448 1.5829 -6.7148>, rot=<0.6161 0.347 0.619 0.3419>}
 dyn_hair_front_B_01_lf {parent=base_hair_front_B_lf, loc=<0 0 0>, rot=<0 0 0 1>}
 base_hair_front_A_lf {parent=head, loc=<3.1628 4.5124 -7.211>, rot=<0.5893 0.3909 0.5925 0.386>}
 dyn_hair_front_A_01_lf {parent=base_hair_front_A_lf, loc=<0 0 0>, rot=<0 0 0 1>}
 base_hair_front_cr {parent=head, loc=<9.7834 9.1533 2.0182>, rot=<-0.0029 0.7071 0.7071 -0.0029>}
 dyn_hair_front_01_cr {parent=base_hair_front_cr, loc=<0 0 0>, rot=<0 0 0 1>}
 base_ears_lf {parent=head, loc=<12.2887 0.2353 -6.3043>, rot=<-0.7071 -0.0029 -0.0029 -0.7071>}
 dyn_ears_lf {parent=base_ears_lf, loc=<0 0 0>, rot=<0 0 0 1>}
 base_ears_rt {parent=head, loc=<12.2887 0.2353 6.304>, rot=<-0.7071 -0.0029 -0.0029 -0.7071>}
 dyn_ears_rt {parent=base_ears_rt, loc=<0 0 0>, rot=<0 0 0 1>}
 FACIAL_C_FacialRoot {parent=head, loc=<2.1524 -0.746 -0.0002>, rot=<0.4879 -0.5118 -0.4879 -0.5118>}
 FACIAL_R_UnderChin {parent=FACIAL_C_FacialRoot, loc=<-2.894 8.2308 2.3744>, rot=<0.3937 0.3557 -0.0787 -0.8439>}
 FACIAL_C_UnderChin {parent=FACIAL_C_FacialRoot, loc=<0.0014 9.4928 3.3166>, rot=<0.4908 0 0 -0.8713>}
 FACIAL_L_UnderChin {parent=FACIAL_C_FacialRoot, loc=<2.8963 8.2364 2.3722>, rot=<0.3938 -0.3557 0.0786 -0.844>}
 FACIAL_C_Jaw {parent=FACIAL_C_FacialRoot, loc=<-0.0003 3.3799 1.3877>, rot=<0.3322 0 -0.0001 -0.9432>}
 FACIAL_C_Jawline1 {parent=FACIAL_C_Jaw, loc=<0.0013 2.5754 6.9229>, rot=<0.2586 0 0 -0.966>}
 base_hair_back_rt {parent=head, loc=<11.7696 -5.4045 5.1862>, rot=<0.0375 -0.9993 0 0>}
 dyn_hair_back_01_rt {parent=base_hair_back_rt, loc=<0 0 0>, rot=<0 0 0 1>}
 dyn_hair_back_02_rt {parent=dyn_hair_back_01_rt, loc=<6.829 0 0>, rot=<0 0 0 1>}
 dyn_hair_back_03_rt {parent=dyn_hair_back_02_rt, loc=<6.829 0 0>, rot=<0 0 0 1>}
 base_hair_side_rt {parent=dyn_hair_back_03_rt, loc=<-2.8865 5.0731 -4.6956>, rot=<0.705 -0.0549 -0.6997 0.1018>}
 dyn_hair_side_rt {parent=base_hair_side_rt, loc=<0 0 0>, rot=<0 0 0 1>}
 base_hair_back_C_cr {parent=dyn_hair_back_03_rt, loc=<-1.5534 -1.6952 -1.2087>, rot=<-0.5164 0.483 0.483 -0.5164>}
 dyn_hair_back_C_01_cr {parent=base_hair_back_C_cr, loc=<0 0 0>, rot=<0 0 0 1>}
 base_hair_back_cr {parent=head, loc=<11.7696 -5.4045 0>, rot=<0.0375 -0.9993 0 0>}
 dyn_hair_back_01_cr {parent=base_hair_back_cr, loc=<0 0 0>, rot=<0 0 0 1>}
 dyn_hair_back_02_cr {parent=dyn_hair_back_01_cr, loc=<6.829 0 0>, rot=<0 0 0 1>}
 dyn_hair_back_03_cr {parent=dyn_hair_back_02_cr, loc=<6.829 0 0>, rot=<0 0 0 1>}
 base_hair_back_B_lf {parent=dyn_hair_back_03_cr, loc=<-0.4478 -1.9465 -0.8118>, rot=<-0.5164 0.483 0.483 -0.5164>}
 dyn_hair_back_B_01_lf {parent=base_hair_back_B_lf, loc=<0 0 0>, rot=<0 0 0 1>}
 base_hair_back_lf {parent=head, loc=<11.7696 -5.4045 -5.186>, rot=<0.0375 -0.9993 0 0>}
 dyn_hair_back_01_lf {parent=base_hair_back_lf, loc=<0 0 0>, rot=<0 0 0 1>}
 dyn_hair_back_02_lf {parent=dyn_hair_back_01_lf, loc=<6.829 0 0>, rot=<0 0 0 1>}
 dyn_hair_back_03_lf {parent=dyn_hair_back_02_lf, loc=<6.829 0 0>, rot=<0 0 0 1>}
 base_hair_side_lf {parent=dyn_hair_back_03_lf, loc=<0.3352 5.16 2.6827>, rot=<-0.0972 0.7004 0.0502 -0.7053>}
 dyn_hair_side_lf {parent=base_hair_side_lf, loc=<0 0 0>, rot=<0 0 0 1>}
 base_hair_back_A_lf {parent=dyn_hair_back_03_lf, loc=<-1.8324 -1.1153 0.8843>, rot=<-0.3492 0.6149 0.3074 -0.6368>}
 dyn_hair_back_A_lf {parent=base_hair_back_A_lf, loc=<0 0 0>, rot=<0 0 0 1>}
 base_hair_back_D_lf {parent=dyn_hair_back_03_lf, loc=<-2.0207 1.6997 2.795>, rot=<-0.1231 0.6963 0.0764 -0.703>}
 dyn_hair_back_D_lf {parent=base_hair_back_D_lf, loc=<0 0 0>, rot=<0 0 0 1>}
 base_hair_back_C_lf {parent=dyn_hair_back_03_lf, loc=<-1.3521 -3.2424 -2.3364>, rot=<-0.3492 0.6149 0.3074 -0.6368>}
 dyn_hair_back_C_lf {parent=base_hair_back_C_lf, loc=<0 0 0>, rot=<0 0 0 1>}
 base_hair_front_side_rt {parent=head, loc=<-0.4464 0.9992 9.2804>, rot=<0.6569 0.2618 0.659 0.2563>}
 dyn_hair_front_side_rt {parent=base_hair_front_side_rt, loc=<0 0 0>, rot=<0 0 0 1>}
 base_neck_bell_cr {parent=neck_01, loc=<0.5878 5.3628 0>, rot=<0 0 -0.5842 -0.8116>}
 dyn_neck_bell_cr {parent=base_neck_bell_cr, loc=<0 0 0>, rot=<0 0 0 1>}
 clavicle_l {parent=spine_05, loc=<5.4831 -0.1053 -3.2946>, rot=<0.0465 -0.7289 -0.009 -0.6829>}
 upperarm_l {parent=clavicle_l, loc=<9.8658 -1.2136 -0.4105>, rot=<0.0164 -0.3245 -0.0269 -0.9454>}
 clavicle_r {parent=spine_05, loc=<5.4831 -0.1053 3.2946>, rot=<0.7289 0.0465 -0.6829 0.009>}
 upperarm_r {parent=clavicle_r, loc=<-9.8658 1.2136 0.4105>, rot=<0.0164 -0.3245 -0.0269 -0.9454>}
 head_null {parent=spine_05, type=Null, loc=<20.6573 3.5489 0>, rot=<0 0 -0.036 -0.9994>}
 clavicle_r_null {parent=spine_05, type=Null, loc=<5.483 -0.1053 3.2946>, rot=<0.7289 0.0465 -0.6829 0.009>}
]
variables [head_up {type=double, default=-6}]
